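# Physics asset for the skirt of the medium male LatteStir Vita cosmetic body, named as its sliding variant.
# Eight chains around the skirt each link five dynamic bone capsules to a small base anchor capsule.
# Each chain's root joint rotates freely, while lower joints have tight swing and twist limits.
# Thigh, calf and foot capsules add leg shapes, linked per leg by freely rotating constraints.
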PhysicsAsset /BRCosmetics/Characters/Player/Male/Medium/Bodies/M_MED_LatteStir_Vita/Meshes/M_MED_LatteSitr_Vita_SkirtSliding_Physics
bodies {
 base_skirt_01_bck [{shapeType=Capsule, length=0.505, radius=0.505, shapeName=base_skirt_01_bck_capsule, center=<0 0 8>, rotation=<rot 0 -90 0>}]
 base_skirt_01_bck_l [{shapeType=Capsule, length=0.505, radius=0.505, shapeName=base_skirt_01_bck_l_capsule, center=<0 0 8>, rotation=<rot 0 -90 0>}]
 base_skirt_01_bck_r [{shapeType=Capsule, length=0.505, radius=0.505, shapeName=base_skirt_01_bck_r_capsule, center=<0 0 8>, rotation=<rot 0 -90 0>}]
 base_skirt_01_fr [{shapeType=Capsule, length=0.505, radius=0.505, shapeName=base_skirt_01_fr_capsule, center=<0 0 8>, rotation=<rot 0 -90 0>}]
 base_skirt_01_front_l [{shapeType=Capsule, length=0.505, radius=0.505, shapeName=base_skirt_01_front_l_capsule, center=<0 0 8>, rotation=<rot 0 -90 0>}]
 base_skirt_01_front_r [{shapeType=Capsule, length=0.505, radius=0.505, shapeName=base_skirt_01_front_r_capsule, center=<0 0 8>, rotation=<rot 0 -90 0>}]
 base_skirt_01_l [{shapeType=Capsule, length=0.505, radius=0.505, shapeName=base_skirt_01_l_capsule, center=<0 0 8>, rotation=<rot 0 -90 0>}]
 base_skirt_01_r [{shapeType=Capsule, length=0.505, radius=0.505, shapeName=base_skirt_01_r_capsule, center=<0 0 8>, rotation=<rot 0 -90 0>}]
 calf_l [{shapeType=Capsule, length=29.8986, radius=8.4157, shapeName=calf_r_capsule, center=<-17.1071 -0.9251 0>, rotation=<rot 90 0 180>}]
 calf_r [{shapeType=Capsule, length=28.2903, radius=8.2756, shapeName=calf_r_capsule, center=<19.6438 2.5824 0>, rotation=<rot 90 0 0>}]
 dyn_skirt_01_bck [{
   shapeType=Capsule
   length=5.0867
   radius=3.0155
   shapeName=dyn_skirt_01_bck_capsule
   center=<2.8673 -0.377 8>
   rotation=<rot 0 1.1461 -89.5263>
  }]
 dyn_skirt_01_bck_l [{
   shapeType=Capsule
   length=2.8407
   radius=2.8399
   shapeName=dyn_skirt_01_bck_l_capsule
   center=<2.5452 -0.6677 8>
   rotation=<rot 7.8743 2.3056 -90>
  }]
 dyn_skirt_01_bck_r [{
   shapeType=Capsule
   length=4.1268
   radius=3.9262
   shapeName=dyn_skirt_01_bck_r_capsule
   center=<7 1.0544 8>
   rotation=<rot 0 -140.1847 -86.0473>
  }]
 dyn_skirt_01_fr [{
   shapeType=Capsule
   length=4.5225
   radius=3.0325
   shapeName=dyn_skirt_01_fr_capsule
   center=<2.6818 0.5037 8>
   rotation=<rot 0 176.2008 -88.2878>
  }]
 dyn_skirt_01_front_l [{
   shapeType=Capsule
   length=3.5663
   radius=3.1863
   shapeName=dyn_skirt_01_front_l_capsule
   center=<2.6009 0.0331 8>
   rotation=<rot 0 -116.4178 -88.5244>
  }]
 dyn_skirt_01_front_r [{
   shapeType=Capsule
   length=4.2438
   radius=3.0454
   shapeName=dyn_skirt_01_front_r_capsule
   center=<2.6357 -0.3089 8>
   rotation=<rot 0 8.892 -80.7251>
  }]
 dyn_skirt_01_l [{
   shapeType=Capsule
   length=4.7351
   radius=3.3018
   shapeName=dyn_skirt_01_l_capsule
   center=<2.7402 -0.5356 8>
   rotation=<rot 0 8.1389 -88.566>
  }]
 dyn_skirt_01_r [{
   shapeType=Capsule
   length=4.213
   radius=3.3259
   shapeName=dyn_skirt_01_r_capsule
   center=<2.8433 0.1304 8>
   rotation=<rot 0 12.7651 -89.9137>
  }]
 dyn_skirt_02_bck [{shapeType=Capsule, length=10, radius=10, shapeName=dyn_skirt_02_bck_capsule, center=<0 0 8>, rotation=<rot 0 0 90>}]
 dyn_skirt_02_bck_l [{shapeType=Capsule, length=10, radius=10, shapeName=dyn_skirt_02_bck_l_capsule, center=<0 0 8>, rotation=<rot 0 0 90>}]
 dyn_skirt_02_bck_r [{shapeType=Capsule, length=10, radius=10, shapeName=dyn_skirt_02_bck_r_capsule, center=<7 0 8>, rotation=<rot 0 0 90>}]
 dyn_skirt_02_fr [{shapeType=Capsule, length=10, radius=10, shapeName=dyn_skirt_02_fr_capsule, center=<0 0 8>, rotation=<rot 0 0 90>}]
 dyn_skirt_02_front_l [{shapeType=Capsule, length=10, radius=10, shapeName=dyn_skirt_02_front_l_capsule, center=<7 0 8>, rotation=<rot 0.0001 0 99.1344>}]
 dyn_skirt_02_front_r [{shapeType=Capsule, length=10, radius=10, shapeName=dyn_skirt_02_front_r_capsule, center=<0 0 8>, rotation=<rot 0 0 90>}]
 dyn_skirt_02_l [{shapeType=Capsule, length=10, radius=10, shapeName=dyn_skirt_02_l_capsule, center=<0 0 8>, rotation=<rot 0 0 90>}]
 dyn_skirt_02_r [{shapeType=Capsule, length=10, radius=10, shapeName=dyn_skirt_02_r_capsule, center=<0 0 8>, rotation=<rot 0 0 90>}]
 dyn_skirt_03_bck [{shapeType=Capsule, length=10, radius=10, shapeName=dyn_skirt_03_bck_capsule, center=<7 0 8>, rotation=<rot 0 0 90>}]
 dyn_skirt_03_bck_l [{shapeType=Capsule, length=10, radius=10, shapeName=dyn_skirt_03_bck_l_capsule, center=<7 0 8>, rotation=<rot 0 0 90>}]
 dyn_skirt_03_bck_r [{shapeType=Capsule, length=10, radius=10, shapeName=dyn_skirt_03_bck_r_capsule, center=<7 0 8>, rotation=<rot 0 0 90>}]
 dyn_skirt_03_fr [{shapeType=Capsule, length=10, radius=10, shapeName=dyn_skirt_03_fr_capsule, center=<7 0 8>, rotation=<rot -8.5597 0 90>}]
 dyn_skirt_03_front_l [{shapeType=Capsule, length=12.6645, radius=7.3355, shapeName=dyn_skirt_03_front_l_capsule, center=<7 0 5.5823>, rotation=<rot 0 0 90>}]
 dyn_skirt_03_front_r [{shapeType=Capsule, length=10, radius=10, shapeName=dyn_skirt_03_front_r_capsule, center=<7 0 8>, rotation=<rot 0 0 90>}]
 dyn_skirt_03_l [{shapeType=Capsule, length=10, radius=10, shapeName=dyn_skirt_03_l_capsule, center=<7 0 8>, rotation=<rot 0 0 90>}]
 dyn_skirt_03_r [{shapeType=Capsule, length=10, radius=10, shapeName=dyn_skirt_03_r_capsule, center=<7 0 8>, rotation=<rot 0 0 90>}]
 dyn_skirt_04_bck [{shapeType=Capsule, length=10, radius=10, shapeName=dyn_skirt_04_bck_capsule, center=<7 0 8>, rotation=<rot 0 0 90>}]
 dyn_skirt_04_bck_l [{shapeType=Capsule, length=10, radius=10, shapeName=dyn_skirt_04_bck_l_capsule, center=<7 0 8>, rotation=<rot 0 0 90>}]
 dyn_skirt_04_bck_r [{shapeType=Capsule, length=10, radius=10, shapeName=dyn_skirt_04_bck_r_capsule, center=<7 0 8>, rotation=<rot 0 0 90>}]
 dyn_skirt_04_fr [{shapeType=Capsule, length=10, radius=10, shapeName=dyn_skirt_04_fr_capsule, center=<7 0 8>, rotation=<rot -8.5597 0 90>}]
 dyn_skirt_04_front_l [{shapeType=Capsule, length=12.6645, radius=7.3355, shapeName=dyn_skirt_04_front_l_capsule, center=<7 0 5.5823>, rotation=<rot 0 0 90>}]
 dyn_skirt_04_front_r [{shapeType=Capsule, length=10, radius=10, shapeName=dyn_skirt_04_front_r_capsule, center=<7 0 8>, rotation=<rot 0 0 90>}]
 dyn_skirt_04_l [{shapeType=Capsule, length=10, radius=10, shapeName=dyn_skirt_04_l_capsule, center=<7 0 8>, rotation=<rot 0 0 90>}]
 dyn_skirt_04_r [{shapeType=Capsule, length=10, radius=10, shapeName=dyn_skirt_04_r_capsule, center=<7 0 8>, rotation=<rot 0 0 90>}]
 dyn_skirt_05_bck [{shapeType=Capsule, length=10, radius=10, shapeName=dyn_skirt_05_bck_capsule, center=<7 0 8>, rotation=<rot 0 0 90>}]
 dyn_skirt_05_bck_l [{shapeType=Capsule, length=10, radius=10, shapeName=dyn_skirt_05_bck_l_capsule, center=<7 0 8>, rotation=<rot 0 0 90>}]
 dyn_skirt_05_bck_r [{shapeType=Capsule, length=10, radius=10, shapeName=dyn_skirt_05_bck_r_capsule, center=<7 0 8>, rotation=<rot 0 0 90>}]
 dyn_skirt_05_fr [{shapeType=Capsule, length=10, radius=10, shapeName=dyn_skirt_05_fr_capsule, center=<9.7827 0 7.5812>, rotation=<rot -8.5597 0 90>}]
 dyn_skirt_05_front_l [{shapeType=Capsule, length=12.6645, radius=7.3319, shapeName=dyn_skirt_05_front_l_capsule, center=<11.476 0 5.5823>, rotation=<rot 0 0 90>}]
 dyn_skirt_05_front_r [{shapeType=Capsule, length=10, radius=10, shapeName=dyn_skirt_05_front_r_capsule, center=<11.3038 0 8>, rotation=<rot 0 0 90>}]
 dyn_skirt_05_l [{shapeType=Capsule, length=10, radius=10, shapeName=dyn_skirt_05_l_capsule, center=<7 0 8>, rotation=<rot 0 0 90>}]
 dyn_skirt_05_r [{shapeType=Capsule, length=10, radius=10, shapeName=dyn_skirt_05_r_capsule, center=<7 0 8>, rotation=<rot 0 0 90>}]
 foot_l [{shapeType=Capsule, length=11.9416, radius=11.6781, shapeName=foot_r_capsule, center=<-1.2859 7.1352 0>, rotation=<rot 0 0 -90>}]
 foot_r [{shapeType=Capsule, length=11.9416, radius=11.6781, shapeName=foot_r_capsule, center=<1.2859 -7.1352 0>, rotation=<rot 0 0 90>}]
 thigh_l [{shapeType=Capsule, length=33.2041, radius=7.5343, shapeName=thigh_r_capsule, center=<-21.0666 0 0>, rotation=<rot 90 0 180>}]
 thigh_r [{shapeType=Capsule, length=33.2041, radius=7.5343, shapeName=thigh_r_capsule, center=<21.0666 0 0>, rotation=<rot 90 0 0>}]
}
constraints [
 {
  bone1Name=dyn_skirt_01_bck
  bone2Name=base_skirt_01_bck
  swing1LimitDegrees=45
  swing1Motion=Free
  swing2LimitDegrees=45
  swing2Motion=Free
  twistLimitDegrees=45
  twistMotion=Free
 }
 {bone1Name=dyn_skirt_02_bck, bone2Name=dyn_skirt_01_bck, swing1LimitDegrees=5, swing2LimitDegrees=30, twistLimitDegrees=0}
 {bone1Name=dyn_skirt_03_bck, bone2Name=dyn_skirt_02_bck, swing1LimitDegrees=5, swing2LimitDegrees=20, twistLimitDegrees=5}
 {bone1Name=dyn_skirt_04_bck, bone2Name=dyn_skirt_03_bck, swing1LimitDegrees=5, swing2LimitDegrees=20, twistLimitDegrees=5}
 {bone1Name=dyn_skirt_05_bck, bone2Name=dyn_skirt_04_bck, swing1LimitDegrees=15, swing2LimitDegrees=20, twistLimitDegrees=5}
 {
  bone1Name=dyn_skirt_01_bck_l
  bone2Name=base_skirt_01_bck_l
  swing1LimitDegrees=45
  swing1Motion=Free
  swing2LimitDegrees=45
  swing2Motion=Free
  twistLimitDegrees=45
  twistMotion=Free
 }
 {bone1Name=dyn_skirt_02_bck_l, bone2Name=dyn_skirt_01_bck_l, swing1LimitDegrees=5, swing2LimitDegrees=30, twistLimitDegrees=0}
 {bone1Name=dyn_skirt_03_bck_l, bone2Name=dyn_skirt_02_bck_l, swing1LimitDegrees=5, swing2LimitDegrees=20, twistLimitDegrees=5}
 {bone1Name=dyn_skirt_04_bck_l, bone2Name=dyn_skirt_03_bck_l, swing1LimitDegrees=5, swing2LimitDegrees=20, twistLimitDegrees=5}
 {bone1Name=dyn_skirt_05_bck_l, bone2Name=dyn_skirt_04_bck_l, swing1LimitDegrees=15, swing2LimitDegrees=20, twistLimitDegrees=5}
 {
  bone1Name=dyn_skirt_01_bck_r
  bone2Name=base_skirt_01_bck_r
  swing1LimitDegrees=45
  swing1Motion=Free
  swing2LimitDegrees=45
  swing2Motion=Free
  twistLimitDegrees=45
  twistMotion=Free
 }
 {bone1Name=dyn_skirt_02_bck_r, bone2Name=dyn_skirt_01_bck_r, swing1LimitDegrees=5, swing2LimitDegrees=30, twistLimitDegrees=0}
 {bone1Name=dyn_skirt_03_bck_r, bone2Name=dyn_skirt_02_bck_r, swing1LimitDegrees=5, swing2LimitDegrees=20, twistLimitDegrees=5}
 {bone1Name=dyn_skirt_04_bck_r, bone2Name=dyn_skirt_03_bck_r, swing1LimitDegrees=5, swing2LimitDegrees=20, twistLimitDegrees=5}
 {bone1Name=dyn_skirt_05_bck_r, bone2Name=dyn_skirt_04_bck_r, swing1LimitDegrees=15, swing2LimitDegrees=20, twistLimitDegrees=5}
 {
  bone1Name=dyn_skirt_01_fr
  bone2Name=base_skirt_01_fr
  swing1LimitDegrees=45
  swing1Motion=Free
  swing2LimitDegrees=45
  swing2Motion=Free
  twistLimitDegrees=45
  twistMotion=Free
 }
 {bone1Name=dyn_skirt_02_fr, bone2Name=dyn_skirt_01_fr, swing1LimitDegrees=5, swing2LimitDegrees=30, twistLimitDegrees=0}
 {bone1Name=dyn_skirt_03_fr, bone2Name=dyn_skirt_02_fr, swing1LimitDegrees=5, swing2LimitDegrees=20, twistLimitDegrees=5}
 {bone1Name=dyn_skirt_04_fr, bone2Name=dyn_skirt_03_fr, swing1LimitDegrees=5, swing2LimitDegrees=20, twistLimitDegrees=5}
 {bone1Name=dyn_skirt_05_fr, bone2Name=dyn_skirt_04_fr, swing1LimitDegrees=15, swing2LimitDegrees=20, twistLimitDegrees=5}
 {
  bone1Name=dyn_skirt_01_front_l
  bone2Name=base_skirt_01_front_l
  swing1LimitDegrees=45
  swing1Motion=Free
  swing2LimitDegrees=45
  swing2Motion=Free
  twistLimitDegrees=45
  twistMotion=Free
 }
 {bone1Name=dyn_skirt_02_front_l, bone2Name=dyn_skirt_01_front_l, swing1LimitDegrees=5, swing2LimitDegrees=30, twistLimitDegrees=0}
 {bone1Name=dyn_skirt_03_front_l, bone2Name=dyn_skirt_02_front_l, swing1LimitDegrees=5, swing2LimitDegrees=20, twistLimitDegrees=5}
 {bone1Name=dyn_skirt_04_front_l, bone2Name=dyn_skirt_03_front_l, swing1LimitDegrees=5, swing2LimitDegrees=20, twistLimitDegrees=5}
 {bone1Name=dyn_skirt_05_front_l, bone2Name=dyn_skirt_04_front_l, swing1LimitDegrees=5, swing2LimitDegrees=20, twistLimitDegrees=5}
 {
  bone1Name=dyn_skirt_01_front_r
  bone2Name=base_skirt_01_front_r
  swing1LimitDegrees=45
  swing1Motion=Free
  swing2LimitDegrees=45
  swing2Motion=Free
  twistLimitDegrees=45
  twistMotion=Free
 }
 {bone1Name=dyn_skirt_02_front_r, bone2Name=dyn_skirt_01_front_r, swing1LimitDegrees=5, swing2LimitDegrees=30, twistLimitDegrees=0}
 {bone1Name=dyn_skirt_03_front_r, bone2Name=dyn_skirt_02_front_r, swing1LimitDegrees=5, swing2LimitDegrees=20, twistLimitDegrees=5}
 {bone1Name=dyn_skirt_04_front_r, bone2Name=dyn_skirt_03_front_r, swing1LimitDegrees=5, swing2LimitDegrees=20, twistLimitDegrees=5}
 {bone1Name=dyn_skirt_05_front_r, bone2Name=dyn_skirt_04_front_r, swing1LimitDegrees=15, swing2LimitDegrees=20, twistLimitDegrees=5}
 {
  bone1Name=dyn_skirt_01_l
  bone2Name=base_skirt_01_l
  swing1LimitDegrees=45
  swing1Motion=Free
  swing2LimitDegrees=45
  swing2Motion=Free
  twistLimitDegrees=45
  twistMotion=Free
 }
 {bone1Name=dyn_skirt_02_l, bone2Name=dyn_skirt_01_l, swing1LimitDegrees=5, swing2LimitDegrees=30, twistLimitDegrees=0}
 {bone1Name=dyn_skirt_03_l, bone2Name=dyn_skirt_02_l, swing1LimitDegrees=5, swing2LimitDegrees=20, twistLimitDegrees=5}
 {bone1Name=dyn_skirt_04_l, bone2Name=dyn_skirt_03_l, swing1LimitDegrees=5, swing2LimitDegrees=20, twistLimitDegrees=5}
 {bone1Name=dyn_skirt_05_l, bone2Name=dyn_skirt_04_l, swing1LimitDegrees=15, swing2LimitDegrees=20, twistLimitDegrees=5}
 {
  bone1Name=dyn_skirt_01_r
  bone2Name=base_skirt_01_r
  swing1LimitDegrees=45
  swing1Motion=Free
  swing2LimitDegrees=45
  swing2Motion=Free
  twistLimitDegrees=45
  twistMotion=Free
 }
 {bone1Name=dyn_skirt_02_r, bone2Name=dyn_skirt_01_r, swing1LimitDegrees=5, swing2LimitDegrees=30, twistLimitDegrees=0}
 {bone1Name=dyn_skirt_03_r, bone2Name=dyn_skirt_02_r, swing1LimitDegrees=5, swing2LimitDegrees=20, twistLimitDegrees=5}
 {bone1Name=dyn_skirt_04_r, bone2Name=dyn_skirt_03_r, swing1LimitDegrees=5, swing2LimitDegrees=20, twistLimitDegrees=5}
 {bone1Name=dyn_skirt_05_r, bone2Name=dyn_skirt_04_r, swing1LimitDegrees=15, swing2LimitDegrees=20, twistLimitDegrees=5}
 {
  bone1Name=foot_r
  bone2Name=calf_r
  swing1LimitDegrees=45
  swing1Motion=Free
  swing2LimitDegrees=45
  swing2Motion=Free
  twistLimitDegrees=45
  twistMotion=Free
 }
 {
  bone1Name=calf_r
  bone2Name=thigh_r
  swing1LimitDegrees=45
  swing1Motion=Free
  swing2LimitDegrees=45
  swing2Motion=Free
  twistLimitDegrees=45
  twistMotion=Free
 }
 {
  bone1Name=foot_l
  bone2Name=calf_l
  swing1LimitDegrees=45
  swing1Motion=Free
  swing2LimitDegrees=45
  swing2Motion=Free
  twistLimitDegrees=45
  twistMotion=Free
 }
 {
  bone1Name=calf_l
  bone2Name=thigh_l
  swing1LimitDegrees=45
  swing1Motion=Free
  swing2LimitDegrees=45
  swing2Motion=Free
  twistLimitDegrees=45
  twistMotion=Free
 }
]
