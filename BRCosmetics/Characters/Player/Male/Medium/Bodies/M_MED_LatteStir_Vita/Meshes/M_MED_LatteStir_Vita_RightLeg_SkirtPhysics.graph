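# Physics asset for the right-side skirt of the male medium LatteStir Vita character body, likely for secondary-motion simulation.
# Front, side and back chains each link five capsule bodies from a base bone, with a free first joint, then tight swing limits and zero twist.
# Leg, foot, pelvis and oversized root capsules provide body shapes.
# Only the back base and left leg are constrained to the pelvis.
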PhysicsAsset /BRCosmetics/Characters/Player/Male/Medium/Bodies/M_MED_LatteStir_Vita/Meshes/M_MED_LatteStir_Vita_RightLeg_SkirtPhysics
bodies {
 base_skirt_01_bck_r [{shapeType=Capsule, length=0.505, radius=0.505, shapeName=base_skirt_01_bck_r_capsule, rotation=<rot 0 -90 0>}]
 base_skirt_01_front_r [{shapeType=Capsule, length=10, radius=5, shapeName=base_skirt_01_front_r_capsule, rotation=<rot 90 0 0>}]
 base_skirt_01_r [{shapeType=Capsule, length=10, radius=5, shapeName=base_skirt_01_r_capsule, rotation=<rot 90 0 0>}]
 calf_l [{
   shapeType=Capsule
   length=39.8339
   radius=7.1999
   shapeName=calf_r_capsule
   center=<-19.8575 -2.3886 0>
   rotation=<rot 87.7521 179.9999 -0.0001>
  }]
 calf_r [{
   shapeType=Capsule
   length=38.5692
   radius=9.1035
   shapeName=calf_r_capsule
   center=<19.8055 1.1306 -0.002>
   rotation=<rot 87.7521 179.9998 -176.1331>
  }]
 dyn_skirt_01_bck_r [{shapeType=Capsule, length=9.213, radius=13.2299, shapeName=dyn_skirt_03_front_r_capsule, center=<12.3835 0 12.0462>, rotation=<rot 0 0 90>}]
 dyn_skirt_01_front_r [{shapeType=Capsule, length=10, radius=5, shapeName=dyn_skirt_01_front_r_capsule, rotation=<rot 90 0 0>}]
 dyn_skirt_01_r [{shapeType=Capsule, length=10, radius=5, shapeName=dyn_skirt_01_r_capsule, rotation=<rot 90 0 0>}]
 dyn_skirt_02_bck_r [{shapeType=Capsule, length=9.213, radius=13.2299, shapeName=dyn_skirt_03_front_r_capsule, center=<12.3835 0 12.0462>, rotation=<rot 0 0 90>}]
 dyn_skirt_02_front_r [{shapeType=Capsule, length=5.5431, radius=9.4569, shapeName=dyn_skirt_02_front_r_capsule, center=<7 0 8.3>, rotation=<rot 90 0 0>}]
 dyn_skirt_02_r [{shapeType=Capsule, length=29.831, radius=13.2299, shapeName=dyn_skirt_03_r_capsule, center=<9.2 0 10.6725>, rotation=<rot 0 0 90>}]
 dyn_skirt_03_bck_r [{shapeType=Capsule, length=27.1874, radius=13.2299, shapeName=dyn_skirt_03_front_r_capsule, center=<12.3835 0 12.0462>, rotation=<rot 0 0 90>}]
 dyn_skirt_03_front_r [{shapeType=Capsule, length=9.213, radius=13.2299, shapeName=dyn_skirt_03_front_r_capsule, center=<9.2 0 12.0462>, rotation=<rot 0 0 90>}]
 dyn_skirt_03_r [{shapeType=Capsule, length=29.831, radius=13.2299, shapeName=dyn_skirt_03_r_capsule, center=<9.2 0 10.6725>, rotation=<rot 0 0 90>}]
 dyn_skirt_04_bck_r [{shapeType=Capsule, length=27.1874, radius=13.2299, shapeName=dyn_skirt_03_front_r_capsule, center=<12.3835 0 12.0462>, rotation=<rot 0 0 90>}]
 dyn_skirt_04_front_r [{shapeType=Capsule, length=9.213, radius=13.2299, shapeName=dyn_skirt_04_front_r_capsule, center=<7.1 0 12.0462>, rotation=<rot 0 0 90>}]
 dyn_skirt_04_r [{shapeType=Capsule, length=29.831, radius=13.2299, shapeName=dyn_skirt_04_r_capsule, center=<7.1 0 10.1219>, rotation=<rot 0 0 90>}]
 dyn_skirt_05_bck_r [{shapeType=Capsule, length=27.1874, radius=13.2299, shapeName=dyn_skirt_03_front_r_capsule, center=<12.3835 0 12.0462>, rotation=<rot 0 0 90>}]
 dyn_skirt_05_front_r [{shapeType=Capsule, length=20.2619, radius=13.3823, shapeName=dyn_skirt_05_front_r_capsule, center=<10.8489 0 12.0462>, rotation=<rot 0 0 90>}]
 dyn_skirt_05_r [{shapeType=Capsule, length=40.8799, radius=13.3823, shapeName=dyn_skirt_05_r_capsule, center=<10.8489 0 10.0599>, rotation=<rot 0 0 90>}]
 foot_l [{shapeType=Capsule, length=15.6533, radius=9.1851, shapeName=foot_r_capsule, center=<-2.5065 6.4971 0>, rotation=<rot 0 0 -90>}]
 foot_r [{shapeType=Capsule, length=15.6533, radius=9.1851, shapeName=foot_r_capsule, center=<2.5065 -6.4971 0>, rotation=<rot 0 0 90>}]
 pelvis [{shapeType=Capsule, length=10, radius=5, shapeName=pelvis_capsule, rotation=<rot 90 0 0>}]
 root [{
   shapeType=Capsule
   length=295.7426
   radius=189.0612
   shapeName=root_capsule
   center=<-40.7806 3.7409 -181.2706>
   rotation=<rot -88.8249 89.968 180>
  }]
 thigh_l [{
   shapeType=Capsule
   length=39.449
   radius=7.8997
   shapeName=thigh_r_capsule
   center=<-22.4902 -0.6884 0.8103>
   rotation=<rot 87.7173 179.9993 -0.0004>
  }]
 thigh_r [{
   shapeType=Capsule
   length=39.449
   radius=7.8997
   shapeName=thigh_r_capsule
   center=<22.4902 0.6884 -0.8103>
   rotation=<rot 87.7173 179.9993 179.9996>
  }]
}
constraints [
 {
  bone1Name=dyn_skirt_01_front_r
  bone2Name=base_skirt_01_front_r
  swing1LimitDegrees=45
  swing1Motion=Free
  swing2LimitDegrees=45
  swing2Motion=Free
  twistLimitDegrees=45
  twistMotion=Free
 }
 {bone1Name=dyn_skirt_02_front_r, bone2Name=dyn_skirt_01_front_r, swing1LimitDegrees=0, swing2LimitDegrees=15, twistLimitDegrees=0}
 {bone1Name=dyn_skirt_03_front_r, bone2Name=dyn_skirt_02_front_r, swing1LimitDegrees=30, swing2LimitDegrees=15, twistLimitDegrees=0}
 {bone1Name=dyn_skirt_04_front_r, bone2Name=dyn_skirt_03_front_r, swing1LimitDegrees=10, swing2LimitDegrees=15, twistLimitDegrees=0}
 {bone1Name=dyn_skirt_05_front_r, bone2Name=dyn_skirt_04_front_r, swing1LimitDegrees=10, swing2LimitDegrees=15, twistLimitDegrees=0}
 {
  bone1Name=dyn_skirt_01_r
  bone2Name=base_skirt_01_r
  swing1LimitDegrees=45
  swing1Motion=Free
  swing2LimitDegrees=45
  swing2Motion=Free
  twistLimitDegrees=45
  twistMotion=Free
 }
 {bone1Name=dyn_skirt_02_r, bone2Name=dyn_skirt_01_r, swing1LimitDegrees=15, swing2LimitDegrees=0, twistLimitDegrees=0}
 {bone1Name=dyn_skirt_03_r, bone2Name=dyn_skirt_02_r, swing1LimitDegrees=9, swing2LimitDegrees=20, twistLimitDegrees=0}
 {bone1Name=dyn_skirt_04_r, bone2Name=dyn_skirt_03_r, swing1LimitDegrees=9, swing2LimitDegrees=20, twistLimitDegrees=0}
 {bone1Name=dyn_skirt_05_r, bone2Name=dyn_skirt_04_r, swing1LimitDegrees=9, swing2LimitDegrees=20, twistLimitDegrees=0}
 {
  bone1Name=foot_l
  bone2Name=calf_l
  swing1LimitDegrees=45
  swing1Motion=Free
  swing2LimitDegrees=45
  swing2Motion=Free
  twistLimitDegrees=45
  twistMotion=Free
 }
 {
  bone1Name=calf_l
  bone2Name=thigh_l
  swing1LimitDegrees=45
  swing1Motion=Free
  swing2LimitDegrees=45
  swing2Motion=Free
  twistLimitDegrees=45
  twistMotion=Free
 }
 {
  bone1Name=thigh_l
  bone2Name=pelvis
  swing1LimitDegrees=45
  swing1Motion=Free
  swing2LimitDegrees=45
  swing2Motion=Free
  twistLimitDegrees=45
  twistMotion=Free
 }
 {
  bone1Name=base_skirt_01_bck_r
  bone2Name=pelvis
  swing1LimitDegrees=45
  swing1Motion=Free
  swing2LimitDegrees=45
  swing2Motion=Free
  twistLimitDegrees=45
  twistMotion=Free
 }
 {
  bone1Name=dyn_skirt_01_bck_r
  bone2Name=base_skirt_01_bck_r
  swing1LimitDegrees=45
  swing1Motion=Free
  swing2LimitDegrees=45
  swing2Motion=Free
  twistLimitDegrees=45
  twistMotion=Free
 }
 {bone1Name=dyn_skirt_02_bck_r, bone2Name=dyn_skirt_01_bck_r, swing1LimitDegrees=0, swing2LimitDegrees=10, twistLimitDegrees=0}
 {bone1Name=dyn_skirt_03_bck_r, bone2Name=dyn_skirt_02_bck_r, swing1LimitDegrees=0, swing2LimitDegrees=10, twistLimitDegrees=0}
 {bone1Name=dyn_skirt_04_bck_r, bone2Name=dyn_skirt_03_bck_r, swing1LimitDegrees=0, swing2LimitDegrees=10, twistLimitDegrees=0}
 {bone1Name=dyn_skirt_05_bck_r, bone2Name=dyn_skirt_04_bck_r, swing1LimitDegrees=0, swing2LimitDegrees=10, twistLimitDegrees=0}
]
