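# Post-pose AnimBlueprint for the LatteStir Vita player body cosmetic, adding secondary motion to its skirt, hair and bow flaps.
# The parent mesh pose feeds a state-driven Control Rig, then skirt correction offsets, per-thigh RigidBody simulations, back-panel AnimDynamics and front-centre rotation constraints.
# A cached skirt pose feeds spring-based AnimDynamics for arm/leg hair and bow bones.
# Additive bone tweaks finish the pose, including sprint-gated thigh-volume scaling.
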AnimBlueprint /BRCosmetics/Characters/Player/Male/Medium/Bodies/M_MED_LatteStir_Vita/Meshes/M_MED_LatteStir_Vita_AnimBP
target_skeleton=&/M_MED_LatteStir_Vita_Skeleton
graphs [
 AnimGraph {nodes [
   Copy Pose From Mesh() {bUseAttachedParent=true}
   n1 = Anim Dynamics - Bone: dyn_hairArm_01_l!(
    ComponentPose=Local To Component(LocalPose=Use cached pose 'CRCache_Skirt_Dynamics'())
    Alpha=1
    GravityOverride:Vector="0,0,-980"
    ExternalForce="0,0,0"
   ) -> (Pose) {
    angularDampingOverride=0.9999
    angularSpringConstant=6000
    bAngularSpring=true
    bOverrideAngularDamping=true
    bUseGravityOverride=true
    gravityScale=0.35
    boundBone {boneName=dyn_hairArm_01_l}
    componentAppliedLinearAccClamp=<2000 2000 2000>
    componentLinearAccScale=<0.2 0.2 0.2>
    componentLinearVelScale=<0.2 0.2 0.2>
    gravityOverride {z=-980}
    physicsBodyDefinitions [{
      boundBone {boneName=dyn_hairArm_01_l}
      boxExtents=<15 20 20>
      constraintSetup {angularLimitsMax=<0 25 25>, angularLimitsMin=<0 -25 -25>, angularTarget=<1 0 0>}
      localJointOffset=<4 0 0>
     }]
    simSpaceSettings {maxAngularAcceleration=3, maxAngularVelocity=8, simSpaceAngularAlpha=1}
   }
   n2 = Anim Dynamics - Bone: dyn_hairArm_02_l!(ComponentPose=@n1, Alpha=1, GravityOverride:Vector="0,0,-980", ExternalForce="0,0,0") -> (Pose) {
    angularDampingOverride=0.9999
    angularSpringConstant=6000
    bAngularSpring=true
    bOverrideAngularDamping=true
    bUseGravityOverride=true
    gravityScale=0.35
    boundBone {boneName=dyn_hairArm_02_l}
    componentAppliedLinearAccClamp=<2000 2000 2000>
    componentLinearAccScale=<0.2 0.2 0.2>
    componentLinearVelScale=<0.2 0.2 0.2>
    gravityOverride {z=-980}
    physicsBodyDefinitions [{
      boundBone {boneName=dyn_hairArm_02_l}
      boxExtents=<15 20 20>
      constraintSetup {angularLimitsMax=<0 25 25>, angularLimitsMin=<0 -25 -25>, angularTarget=<1 0 0>}
      localJointOffset=<4 0 0>
     }]
    simSpaceSettings {maxAngularAcceleration=3, maxAngularVelocity=8, simSpaceAngularAlpha=1}
   }
   n3 = Anim Dynamics - Bone: dyn_hairArm_03_l!(ComponentPose=@n2, Alpha=1, GravityOverride:Vector="0,0,-980", ExternalForce="0,0,0") -> (Pose) {
    angularDampingOverride=0.9999
    angularSpringConstant=6000
    bAngularSpring=true
    bOverrideAngularDamping=true
    bUseGravityOverride=true
    gravityScale=0.35
    boundBone {boneName=dyn_hairArm_03_l}
    componentAppliedLinearAccClamp=<2000 2000 2000>
    componentLinearAccScale=<0.2 0.2 0.2>
    componentLinearVelScale=<0.2 0.2 0.2>
    gravityOverride {z=-980}
    physicsBodyDefinitions [{
      boundBone {boneName=dyn_hairArm_03_l}
      boxExtents=<15 20 20>
      constraintSetup {angularLimitsMax=<0 25 25>, angularLimitsMin=<0 -25 -25>, angularTarget=<1 0 0>}
      localJointOffset=<4 0 0>
     }]
    simSpaceSettings {maxAngularAcceleration=3, maxAngularVelocity=8, simSpaceAngularAlpha=1}
   }
   n4 = Anim Dynamics - Bone: dyn_hairArm_04_l!(ComponentPose=@n3, Alpha=1, GravityOverride:Vector="0,0,-980", ExternalForce="0,0,0") -> (Pose) {
    angularDampingOverride=0.9999
    angularSpringConstant=6000
    bAngularSpring=true
    bOverrideAngularDamping=true
    bUseGravityOverride=true
    gravityScale=0.35
    boundBone {boneName=dyn_hairArm_04_l}
    componentAppliedLinearAccClamp=<2000 2000 2000>
    componentLinearAccScale=<0.2 0.2 0.2>
    componentLinearVelScale=<0.2 0.2 0.2>
    gravityOverride {z=-980}
    physicsBodyDefinitions [{
      boundBone {boneName=dyn_hairArm_04_l}
      boxExtents=<15 20 20>
      constraintSetup {angularLimitsMax=<0 25 25>, angularLimitsMin=<0 -25 -25>, angularTarget=<1 0 0>}
      localJointOffset=<4 0 0>
     }]
    simSpaceSettings {maxAngularAcceleration=3, maxAngularVelocity=8, simSpaceAngularAlpha=1}
   }
   n5 = Anim Dynamics - Bone: dyn_hairArm_05_l!(ComponentPose=@n4, Alpha=1, GravityOverride:Vector="0,0,-980", ExternalForce="0,0,0") -> (Pose) {
    angularDampingOverride=0.9999
    angularSpringConstant=6000
    bAngularSpring=true
    bOverrideAngularDamping=true
    bUseGravityOverride=true
    gravityScale=0.35
    boundBone {boneName=dyn_hairArm_05_l}
    componentAppliedLinearAccClamp=<2000 2000 2000>
    componentLinearAccScale=<0.2 0.2 0.2>
    componentLinearVelScale=<0.2 0.2 0.2>
    gravityOverride {z=-980}
    physicsBodyDefinitions [{
      boundBone {boneName=dyn_hairArm_05_l}
      boxExtents=<15 20 20>
      constraintSetup {angularLimitsMax=<0 25 25>, angularLimitsMin=<0 -25 -25>, angularTarget=<1 0 0>}
      localJointOffset=<4 0 0>
     }]
    simSpaceSettings {maxAngularAcceleration=3, maxAngularVelocity=8, simSpaceAngularAlpha=1}
   }
   n6 = Anim Dynamics - Bone: dyn_hairArm_01_r!(ComponentPose=@n5, Alpha=1, GravityOverride:Vector="0,0,-980", ExternalForce="0,0,0") -> (Pose) {
    angularDampingOverride=0.9999
    angularSpringConstant=6000
    bAngularSpring=true
    bOverrideAngularDamping=true
    bUseGravityOverride=true
    gravityScale=0.35
    boundBone {boneName=dyn_hairArm_01_r}
    componentAppliedLinearAccClamp=<2000 2000 2000>
    componentLinearAccScale=<0.2 0.2 0.2>
    componentLinearVelScale=<0.2 0.2 0.2>
    gravityOverride {z=-980}
    physicsBodyDefinitions [{
      boundBone {boneName=dyn_hairArm_01_r}
      boxExtents=<15 20 20>
      constraintSetup {angularLimitsMax=<0 25 25>, angularLimitsMin=<0 -25 -25>, angularTarget=<1 0 0>}
      localJointOffset=<4 0 0>
     }]
    simSpaceSettings {maxAngularAcceleration=3, maxAngularVelocity=8, simSpaceAngularAlpha=1}
   }
   n7 = Anim Dynamics - Bone: dyn_hairArm_02_r!(ComponentPose=@n6, Alpha=1, GravityOverride:Vector="0,0,-980", ExternalForce="0,0,0") -> (Pose) {
    angularDampingOverride=0.9999
    angularSpringConstant=6000
    bAngularSpring=true
    bOverrideAngularDamping=true
    bUseGravityOverride=true
    gravityScale=0.35
    boundBone {boneName=dyn_hairArm_02_r}
    componentAppliedLinearAccClamp=<2000 2000 2000>
    componentLinearAccScale=<0.2 0.2 0.2>
    componentLinearVelScale=<0.2 0.2 0.2>
    gravityOverride {z=-980}
    physicsBodyDefinitions [{
      boundBone {boneName=dyn_hairArm_02_r}
      boxExtents=<15 20 20>
      constraintSetup {angularLimitsMax=<0 25 25>, angularLimitsMin=<0 -25 -25>, angularTarget=<1 0 0>}
      localJointOffset=<4 0 0>
     }]
    simSpaceSettings {maxAngularAcceleration=3, maxAngularVelocity=8, simSpaceAngularAlpha=1}
   }
   n8 = Anim Dynamics - Bone: dyn_hairArm_03_r!(ComponentPose=@n7, Alpha=1, GravityOverride:Vector="0,0,-980", ExternalForce="0,0,0") -> (Pose) {
    angularDampingOverride=0.9999
    angularSpringConstant=6000
    bAngularSpring=true
    bOverrideAngularDamping=true
    bUseGravityOverride=true
    gravityScale=0.35
    boundBone {boneName=dyn_hairArm_03_r}
    componentAppliedLinearAccClamp=<2000 2000 2000>
    componentLinearAccScale=<0.2 0.2 0.2>
    componentLinearVelScale=<0.2 0.2 0.2>
    gravityOverride {z=-980}
    physicsBodyDefinitions [{
      boundBone {boneName=dyn_hairArm_03_r}
      boxExtents=<15 20 20>
      constraintSetup {angularLimitsMax=<0 25 25>, angularLimitsMin=<0 -25 -25>, angularTarget=<1 0 0>}
      localJointOffset=<4 0 0>
     }]
    simSpaceSettings {maxAngularAcceleration=3, maxAngularVelocity=8, simSpaceAngularAlpha=1}
   }
   n9 = Anim Dynamics - Bone: dyn_hairArm_04_r!(ComponentPose=@n8, Alpha=1, GravityOverride:Vector="0,0,-980", ExternalForce="0,0,0") -> (Pose) {
    angularDampingOverride=0.9999
    angularSpringConstant=6000
    bAngularSpring=true
    bOverrideAngularDamping=true
    bUseGravityOverride=true
    gravityScale=0.35
    boundBone {boneName=dyn_hairArm_04_r}
    componentAppliedLinearAccClamp=<2000 2000 2000>
    componentLinearAccScale=<0.2 0.2 0.2>
    componentLinearVelScale=<0.2 0.2 0.2>
    gravityOverride {z=-980}
    physicsBodyDefinitions [{
      boundBone {boneName=dyn_hairArm_04_r}
      boxExtents=<15 20 20>
      constraintSetup {angularLimitsMax=<0 25 25>, angularLimitsMin=<0 -25 -25>, angularTarget=<1 0 0>}
      localJointOffset=<4 0 0>
     }]
    simSpaceSettings {maxAngularAcceleration=3, maxAngularVelocity=8, simSpaceAngularAlpha=1}
   }
   n10 = Anim Dynamics - Bone: dyn_hairArm_05_r!(ComponentPose=@n9, Alpha=1, GravityOverride:Vector="0,0,-980", ExternalForce="0,0,0") -> (Pose) {
    angularDampingOverride=0.9999
    angularSpringConstant=6000
    bAngularSpring=true
    bOverrideAngularDamping=true
    bUseGravityOverride=true
    gravityScale=0.35
    boundBone {boneName=dyn_hairArm_05_r}
    componentAppliedLinearAccClamp=<2000 2000 2000>
    componentLinearAccScale=<0.2 0.2 0.2>
    componentLinearVelScale=<0.2 0.2 0.2>
    gravityOverride {z=-980}
    physicsBodyDefinitions [{
      boundBone {boneName=dyn_hairArm_05_r}
      boxExtents=<15 20 20>
      constraintSetup {angularLimitsMax=<0 25 25>, angularLimitsMin=<0 -25 -25>, angularTarget=<1 0 0>}
      localJointOffset=<4 0 0>
     }]
    simSpaceSettings {maxAngularAcceleration=3, maxAngularVelocity=8, simSpaceAngularAlpha=1}
   }
   n11 = Anim Dynamics - Bone: dyn_hairLeg_01_l!(ComponentPose=@n10, Alpha=1, GravityOverride:Vector="0,0,-980", ExternalForce="0,0,0") -> (Pose) {
    angularDampingOverride=0.9999
    angularSpringConstant=6000
    bAngularSpring=true
    bOverrideAngularDamping=true
    bUseGravityOverride=true
    gravityScale=0.35
    boundBone {boneName=dyn_hairLeg_01_l}
    componentAppliedLinearAccClamp=<2000 2000 2000>
    componentLinearAccScale=<0.2 0.2 0.2>
    componentLinearVelScale=<0.2 0.2 0.2>
    gravityOverride {z=-980}
    physicsBodyDefinitions [{
      boundBone {boneName=dyn_hairLeg_01_l}
      boxExtents=<15 20 20>
      constraintSetup {angularLimitsMax=<0 25 25>, angularLimitsMin=<0 -25 -25>, angularTarget=<1 0 0>}
      localJointOffset=<4 0 0>
     }]
    simSpaceSettings {maxAngularAcceleration=3, maxAngularVelocity=8, simSpaceAngularAlpha=1}
   }
   n12 = Anim Dynamics - Bone: dyn_hairLeg_02_l!(ComponentPose=@n11, Alpha=1, GravityOverride:Vector="0,0,-980", ExternalForce="0,0,0") -> (Pose) {
    angularDampingOverride=0.9999
    angularSpringConstant=6000
    bAngularSpring=true
    bOverrideAngularDamping=true
    bUseGravityOverride=true
    gravityScale=0.35
    boundBone {boneName=dyn_hairLeg_02_l}
    componentAppliedLinearAccClamp=<2000 2000 2000>
    componentLinearAccScale=<0.2 0.2 0.2>
    componentLinearVelScale=<0.2 0.2 0.2>
    gravityOverride {z=-980}
    physicsBodyDefinitions [{
      boundBone {boneName=dyn_hairLeg_02_l}
      boxExtents=<15 20 20>
      constraintSetup {angularLimitsMax=<0 25 25>, angularLimitsMin=<0 -25 -25>, angularTarget=<1 0 0>}
      localJointOffset=<4 0 0>
     }]
    simSpaceSettings {maxAngularAcceleration=3, maxAngularVelocity=8, simSpaceAngularAlpha=1}
   }
   n13 = Anim Dynamics - Bone: dyn_hairLeg_03_l!(ComponentPose=@n12, Alpha=1, GravityOverride:Vector="0,0,-980", ExternalForce="0,0,0") -> (Pose) {
    angularDampingOverride=0.9999
    angularSpringConstant=6000
    bAngularSpring=true
    bOverrideAngularDamping=true
    bUseGravityOverride=true
    gravityScale=0.35
    boundBone {boneName=dyn_hairLeg_03_l}
    componentAppliedLinearAccClamp=<2000 2000 2000>
    componentLinearAccScale=<0.2 0.2 0.2>
    componentLinearVelScale=<0.2 0.2 0.2>
    gravityOverride {z=-980}
    physicsBodyDefinitions [{
      boundBone {boneName=dyn_hairLeg_03_l}
      boxExtents=<15 20 20>
      constraintSetup {angularLimitsMax=<0 25 25>, angularLimitsMin=<0 -25 -25>, angularTarget=<1 0 0>}
      localJointOffset=<4 0 0>
     }]
    simSpaceSettings {maxAngularAcceleration=3, maxAngularVelocity=8, simSpaceAngularAlpha=1}
   }
   n14 = Anim Dynamics - Bone: dyn_hairLeg_04_l!(ComponentPose=@n13, Alpha=1, GravityOverride:Vector="0,0,-980", ExternalForce="0,0,0") -> (Pose) {
    angularDampingOverride=0.9999
    angularSpringConstant=6000
    bAngularSpring=true
    bOverrideAngularDamping=true
    bUseGravityOverride=true
    gravityScale=0.35
    boundBone {boneName=dyn_hairLeg_04_l}
    componentAppliedLinearAccClamp=<2000 2000 2000>
    componentLinearAccScale=<0.2 0.2 0.2>
    componentLinearVelScale=<0.2 0.2 0.2>
    gravityOverride {z=-980}
    physicsBodyDefinitions [{
      boundBone {boneName=dyn_hairLeg_04_l}
      boxExtents=<15 20 20>
      constraintSetup {angularLimitsMax=<0 25 25>, angularLimitsMin=<0 -25 -25>, angularTarget=<1 0 0>}
      localJointOffset=<4 0 0>
     }]
    simSpaceSettings {maxAngularAcceleration=3, maxAngularVelocity=8, simSpaceAngularAlpha=1}
   }
   n15 = Anim Dynamics - Bone: dyn_hairLeg_05_l!(ComponentPose=@n14, Alpha=1, GravityOverride:Vector="0,0,-980", ExternalForce="0,0,0") -> (Pose) {
    angularDampingOverride=0.9999
    angularSpringConstant=6000
    bAngularSpring=true
    bOverrideAngularDamping=true
    bUseGravityOverride=true
    gravityScale=0.35
    boundBone {boneName=dyn_hairLeg_05_l}
    componentAppliedLinearAccClamp=<2000 2000 2000>
    componentLinearAccScale=<0.2 0.2 0.2>
    componentLinearVelScale=<0.2 0.2 0.2>
    gravityOverride {z=-980}
    physicsBodyDefinitions [{
      boundBone {boneName=dyn_hairLeg_05_l}
      boxExtents=<15 20 20>
      constraintSetup {angularLimitsMax=<0 25 25>, angularLimitsMin=<0 -25 -25>, angularTarget=<1 0 0>}
      localJointOffset=<4 0 0>
     }]
    simSpaceSettings {maxAngularAcceleration=3, maxAngularVelocity=8, simSpaceAngularAlpha=1}
   }
   n16 = Anim Dynamics - Bone: dyn_hairLeg_01_r!(ComponentPose=@n15, Alpha=1, GravityOverride:Vector="0,0,-980", ExternalForce="0,0,0") -> (Pose) {
    angularDampingOverride=0.9999
    angularSpringConstant=6000
    bAngularSpring=true
    bOverrideAngularDamping=true
    bUseGravityOverride=true
    gravityScale=0.35
    boundBone {boneName=dyn_hairLeg_01_r}
    componentAppliedLinearAccClamp=<2000 2000 2000>
    componentLinearAccScale=<0.2 0.2 0.2>
    componentLinearVelScale=<0.2 0.2 0.2>
    gravityOverride {z=-980}
    physicsBodyDefinitions [{
      boundBone {boneName=dyn_hairLeg_01_r}
      boxExtents=<15 20 20>
      constraintSetup {angularLimitsMax=<0 25 25>, angularLimitsMin=<0 -25 -25>, angularTarget=<1 0 0>}
      localJointOffset=<4 0 0>
     }]
    simSpaceSettings {maxAngularAcceleration=3, maxAngularVelocity=8, simSpaceAngularAlpha=1}
   }
   n17 = Anim Dynamics - Bone: dyn_hairLeg_02_r!(ComponentPose=@n16, Alpha=1, GravityOverride:Vector="0,0,-980", ExternalForce="0,0,0") -> (Pose) {
    angularDampingOverride=0.9999
    angularSpringConstant=6000
    bAngularSpring=true
    bOverrideAngularDamping=true
    bUseGravityOverride=true
    gravityScale=0.35
    boundBone {boneName=dyn_hairLeg_02_r}
    componentAppliedLinearAccClamp=<2000 2000 2000>
    componentLinearAccScale=<0.2 0.2 0.2>
    componentLinearVelScale=<0.2 0.2 0.2>
    gravityOverride {z=-980}
    physicsBodyDefinitions [{
      boundBone {boneName=dyn_hairLeg_02_r}
      boxExtents=<15 20 20>
      constraintSetup {angularLimitsMax=<0 25 25>, angularLimitsMin=<0 -25 -25>, angularTarget=<1 0 0>}
      localJointOffset=<4 0 0>
     }]
    simSpaceSettings {maxAngularAcceleration=3, maxAngularVelocity=8, simSpaceAngularAlpha=1}
   }
   n18 = Anim Dynamics - Bone: dyn_hairLeg_03_r!(ComponentPose=@n17, Alpha=1, GravityOverride:Vector="0,0,-980", ExternalForce="0,0,0") -> (Pose) {
    angularDampingOverride=0.9999
    angularSpringConstant=6000
    bAngularSpring=true
    bOverrideAngularDamping=true
    bUseGravityOverride=true
    gravityScale=0.35
    boundBone {boneName=dyn_hairLeg_03_r}
    componentAppliedLinearAccClamp=<2000 2000 2000>
    componentLinearAccScale=<0.2 0.2 0.2>
    componentLinearVelScale=<0.2 0.2 0.2>
    gravityOverride {z=-980}
    physicsBodyDefinitions [{
      boundBone {boneName=dyn_hairLeg_03_r}
      boxExtents=<15 20 20>
      constraintSetup {angularLimitsMax=<0 25 25>, angularLimitsMin=<0 -25 -25>, angularTarget=<1 0 0>}
      localJointOffset=<4 0 0>
     }]
    simSpaceSettings {maxAngularAcceleration=3, maxAngularVelocity=8, simSpaceAngularAlpha=1}
   }
   n19 = Anim Dynamics - Bone: dyn_hairLeg_04_r!(ComponentPose=@n18, Alpha=1, GravityOverride:Vector="0,0,-980", ExternalForce="0,0,0") -> (Pose) {
    angularDampingOverride=0.9999
    angularSpringConstant=6000
    bAngularSpring=true
    bOverrideAngularDamping=true
    bUseGravityOverride=true
    gravityScale=0.35
    boundBone {boneName=dyn_hairLeg_04_r}
    componentAppliedLinearAccClamp=<2000 2000 2000>
    componentLinearAccScale=<0.2 0.2 0.2>
    componentLinearVelScale=<0.2 0.2 0.2>
    gravityOverride {z=-980}
    physicsBodyDefinitions [{
      boundBone {boneName=dyn_hairLeg_04_r}
      boxExtents=<15 20 20>
      constraintSetup {angularLimitsMax=<0 25 25>, angularLimitsMin=<0 -25 -25>, angularTarget=<1 0 0>}
      localJointOffset=<4 0 0>
     }]
    simSpaceSettings {maxAngularAcceleration=3, maxAngularVelocity=8, simSpaceAngularAlpha=1}
   }
   n20 = Anim Dynamics - Bone: dyn_hairLeg_05_r!(ComponentPose=@n19, Alpha=1, GravityOverride:Vector="0,0,-980", ExternalForce="0,0,0") -> (Pose) {
    angularDampingOverride=0.9999
    angularSpringConstant=6000
    bAngularSpring=true
    bOverrideAngularDamping=true
    bUseGravityOverride=true
    gravityScale=0.35
    boundBone {boneName=dyn_hairLeg_05_r}
    componentAppliedLinearAccClamp=<2000 2000 2000>
    componentLinearAccScale=<0.2 0.2 0.2>
    componentLinearVelScale=<0.2 0.2 0.2>
    gravityOverride {z=-980}
    physicsBodyDefinitions [{
      boundBone {boneName=dyn_hairLeg_05_r}
      boxExtents=<15 20 20>
      constraintSetup {angularLimitsMax=<0 25 25>, angularLimitsMin=<0 -25 -25>, angularTarget=<1 0 0>}
      localJointOffset=<4 0 0>
     }]
    simSpaceSettings {maxAngularAcceleration=3, maxAngularVelocity=8, simSpaceAngularAlpha=1}
   }
   n21 = Anim Dynamics - Bone: dyn_bowflap_01_r!(ComponentPose=@n20, Alpha=1, GravityOverride:Vector="0,0,-980", ExternalForce="0,0,0") -> (Pose) {
    angularDampingOverride=0.999
    angularSpringConstant=4000
    bAngularSpring=true
    bOverrideAngularDamping=true
    bUseGravityOverride=true
    gravityScale=0.35
    boundBone {boneName=dyn_bowflap_01_r}
    componentAppliedLinearAccClamp=<3000 3000 3000>
    componentLinearAccScale=<0.3 0.3 0.3>
    componentLinearVelScale=<0.3 0.3 0.3>
    gravityOverride {z=-980}
    physicsBodyDefinitions [{
      boundBone {boneName=dyn_bowflap_01_r}
      boxExtents=<15 20 20>
      constraintSetup {angularLimitsMax=<0 0 25>, angularLimitsMin=<0 -25 -5>, angularTarget=<1 0 0>}
      localJointOffset=<4 0 0>
     }]
    simSpaceSettings {maxAngularAcceleration=3, maxAngularVelocity=8, simSpaceAngularAlpha=1}
   }
   n22 = Anim Dynamics - Bone: dyn_bowflap_02_r!(ComponentPose=@n21, Alpha=1, GravityOverride:Vector="0,0,-980", ExternalForce="0,0,0") -> (Pose) {
    angularDampingOverride=0.999
    angularSpringConstant=4000
    bAngularSpring=true
    bOverrideAngularDamping=true
    bUseGravityOverride=true
    gravityScale=0.35
    boundBone {boneName=dyn_bowflap_02_r}
    componentAppliedLinearAccClamp=<3000 3000 3000>
    componentLinearAccScale=<0.3 0.3 0.3>
    componentLinearVelScale=<0.3 0.3 0.3>
    gravityOverride {z=-980}
    physicsBodyDefinitions [{
      boundBone {boneName=dyn_bowflap_02_r}
      boxExtents=<15 20 20>
      constraintSetup {angularLimitsMax=<0 0 25>, angularLimitsMin=<0 -25 -5>, angularTarget=<1 0 0>}
      localJointOffset=<4 0 0>
     }]
    simSpaceSettings {maxAngularAcceleration=3, maxAngularVelocity=8, simSpaceAngularAlpha=1}
   }
   n23 = Anim Dynamics - Bone: dyn_bowflap_01_l!(ComponentPose=@n22, Alpha=1, GravityOverride:Vector="0,0,-980", ExternalForce="0,0,0") -> (Pose) {
    angularDampingOverride=0.999
    angularSpringConstant=4000
    bAngularSpring=true
    bOverrideAngularDamping=true
    bUseGravityOverride=true
    gravityScale=0.35
    boundBone {boneName=dyn_bowflap_01_l}
    componentAppliedLinearAccClamp=<3000 3000 3000>
    componentLinearAccScale=<0.3 0.3 0.3>
    componentLinearVelScale=<0.3 0.3 0.3>
    gravityOverride {z=-980}
    physicsBodyDefinitions [{
      boundBone {boneName=dyn_bowflap_01_l}
      boxExtents=<15 20 20>
      constraintSetup {angularLimitsMax=<0 0 5>, angularLimitsMin=<0 -25 -25>, angularTarget=<1 0 0>}
      localJointOffset=<4 0 0>
     }]
    simSpaceSettings {maxAngularAcceleration=3, maxAngularVelocity=8, simSpaceAngularAlpha=1}
   }
   n24 = Anim Dynamics - Bone: dyn_bowflap_02_l!(ComponentPose=@n23, Alpha=1, GravityOverride:Vector="0,0,-980", ExternalForce="0,0,0") -> (Pose) {
    angularDampingOverride=0.999
    angularSpringConstant=4000
    bAngularSpring=true
    bOverrideAngularDamping=true
    bUseGravityOverride=true
    gravityScale=0.35
    boundBone {boneName=dyn_bowflap_02_l}
    componentAppliedLinearAccClamp=<3000 3000 3000>
    componentLinearAccScale=<0.3 0.3 0.3>
    componentLinearVelScale=<0.3 0.3 0.3>
    gravityOverride {z=-980}
    physicsBodyDefinitions [{
      boundBone {boneName=dyn_bowflap_02_l}
      boxExtents=<15 20 20>
      constraintSetup {angularLimitsMax=<0 0 5>, angularLimitsMin=<0 -25 -25>, angularTarget=<1 0 0>}
      localJointOffset=<4 0 0>
     }]
    simSpaceSettings {maxAngularAcceleration=3, maxAngularVelocity=8, simSpaceAngularAlpha=1}
   }
   n25 = "Transform (Modify) Bone - Bone: dyn_skirt_02_bck_l"!(ComponentPose=@n24, bAlphaBoolEnabled:bool=True, Translation="0,0,0", Rotation="10,0,0", Scale="1,1,1") -> (Pose) {alphaInputType=Bool, rotationMode=BMM_Additive, rotationSpace=BCS_BoneSpace, boneToModify {boneName=dyn_skirt_02_bck_l}, rotation {pitch=10}}
   n26 = "Transform (Modify) Bone - Bone: dyn_skirt_01_bck_l"!(ComponentPose=@n25, bAlphaBoolEnabled:bool=True, Translation="0,0,0", Rotation="55.016,-0.000,-0.000", Scale="1,1,1") -> (Pose) {
    alphaInputType=Bool
    rotationMode=BMM_Additive
    rotationSpace=BCS_BoneSpace
    alphaBoolBlend {blendInTime=0.3, blendOutTime=0.3}
    boneToModify {boneName=dyn_skirt_01_bck_l}
    rotation=<rot 55.016 0 0>
   }
   n27 = "Transform (Modify) Bone - Bone: dyn_skirt_02_l"!(ComponentPose=@n26, bAlphaBoolEnabled:bool=True, Translation="0,0,0", Rotation="16.686,-0.019,-0.001", Scale="1,1,1") -> (Pose) {
    alphaInputType=Bool
    rotationMode=BMM_Additive
    rotationSpace=BCS_BoneSpace
    alphaBoolBlend {blendInTime=0.3, blendOutTime=0.3}
    boneToModify {boneName=dyn_skirt_02_l}
    rotation=<rot 16.686 -0.019 -0.0011>
   }
   n28 = "Transform (Modify) Bone - Bone: base_blackSport"!(ComponentPose=@n27, bAlphaBoolEnabled:bool=True, Translation="0,-40,5", Rotation="0,0,0", Scale="1,1,1") -> (Pose) {
    alphaInputType=Bool
    rotationMode=BMM_Additive
    rotationSpace=BCS_BoneSpace
    translationMode=BMM_Additive
    translationSpace=BCS_BoneSpace
    alphaBoolBlend {blendInTime=0.3, blendOutTime=0.3}
    boneToModify {boneName=base_blackSport}
    translation {y=-40, z=5}
   }
   n29 = "Transform (Modify) Bone - Bone: base_thigh_twist_Volume"!(ComponentPose=@n28, bAlphaBoolEnabled:bool=Get bIsSprinting(), Translation="0,0,0", Rotation="0,0,0", Scale="1,1,0.7") -> (Pose) {
    alphaInputType=Bool
    rotationMode=BMM_Additive
    rotationSpace=BCS_BoneSpace
    scaleMode=BMM_Additive
    scaleSpace=BCS_BoneSpace
    translationMode=BMM_Additive
    translationSpace=BCS_BoneSpace
    alphaBoolBlend {blendInTime=0.3, blendOutTime=0.3}
    boneToModify {boneName=base_thigh_twist_Volume}
    scale {z=0.7}
   }
   Output Pose(Result=Component To Local(ComponentPose=@n29)) {layerGroup=None}
   n30 = Control Rig!(
    Source=Copy Pose From Mesh() {bUseAttachedParent=true}
    Alpha
    isCrouching:double=Get isCrouch() -> (isCrouch:double)
    isMoving:double=Get Moving2d() -> (Moving2d:double)
    inFloating:double=Get isFloating() -> (isFloating:double)
    IsInAir:bool=Get bIsSkyDiving()
    openParachute:double=Get openParachute() -> (openParachute:double)
    IsEmoting:bool=Get bIsPlayingEmote()
    TacticalSprinting:bool=Get bIsTacticalSprinting()
    glider:bool=Get bIsParachuteOpened()
   ) -> (Pose) {controlRigAssetReference {blueprintRigClass=&/M_MED_LatteStir_Vita_CtrlRig.M_MED_LatteStir_Vita_CtrlRig_C}}
   Save cached pose 'CRCache_InputPose'(Pose=@n30) {cacheName=CRCache_InputPose}
   n31 = "Transform (Modify) Bone - Bone: dyn_skirt_01_front_l"!(ComponentPose=Local To Component(LocalPose=Use cached pose 'CRCache_InputPose'()), bAlphaBoolEnabled:bool=True) -> (Pose) {
    alphaInputType=Bool
    rotationMode=BMM_Additive
    rotationSpace=BCS_BoneSpace
    translationMode=BMM_Additive
    translationSpace=BCS_BoneSpace
    alphaBoolBlend {blendInTime=0.2, blendOption=Linear, blendOutTime=0.2}
    boneToModify {boneName=dyn_skirt_01_front_l}
    rotation=<rot 6.1968 0.0238 -0.0128>
   }
   n32 = "Transform (Modify) Bone - Bone: dyn_skirt_01_l"!(ComponentPose=@n31, bAlphaBoolEnabled:bool=True) -> (Pose) {
    alphaInputType=Bool
    rotationMode=BMM_Additive
    rotationSpace=BCS_BoneSpace
    translationMode=BMM_Additive
    translationSpace=BCS_BoneSpace
    alphaBoolBlend {blendInTime=0.2, blendOption=Linear, blendOutTime=0.2}
    boneToModify {boneName=dyn_skirt_01_l}
    rotation=<rot 3.8881 0.0028 -0.0016>
   }
   n33 = "Transform (Modify) Bone - Bone: dyn_skirt_01_front_r"!(ComponentPose=@n32, bAlphaBoolEnabled:bool=True) -> (Pose) {
    alphaInputType=Bool
    rotationMode=BMM_Additive
    rotationSpace=BCS_BoneSpace
    translationMode=BMM_Additive
    translationSpace=BCS_BoneSpace
    alphaBoolBlend {blendInTime=0.2, blendOption=Linear, blendOutTime=0.2}
    boneToModify {boneName=dyn_skirt_01_front_r}
    rotation {pitch=5}
   }
   n34 = "Transform (Modify) Bone - Bone: dyn_skirt_01_r"!(ComponentPose=@n33, bAlphaBoolEnabled:bool=True) -> (Pose) {
    alphaInputType=Bool
    rotationMode=BMM_Additive
    rotationSpace=BCS_BoneSpace
    translationMode=BMM_Additive
    translationSpace=BCS_BoneSpace
    alphaBoolBlend {blendInTime=0.2, blendOption=Linear, blendOutTime=0.2}
    boneToModify {boneName=dyn_skirt_01_r}
    rotation {pitch=5}
   }
   n35 = "Transform (Modify) Bone - Bone: dyn_skirt_01_front_l"(ComponentPose=@n34, bAlphaBoolEnabled=True) {
    alphaInputType=Bool
    rotationMode=BMM_Additive
    rotationSpace=BCS_BoneSpace
    translationMode=BMM_Additive
    translationSpace=BCS_BoneSpace
    alphaBoolBlend {blendInTime=0.2, blendOption=Linear, blendOutTime=0.2}
    boneToModify {boneName=dyn_skirt_01_front_l}
    rotation=<rot 2.3124 0.0407 -0.0151>
   }
   n36 = "Transform (Modify) Bone - Bone: dyn_skirt_01_l"(ComponentPose=@n35, bAlphaBoolEnabled=True) {
    alphaInputType=Bool
    rotationMode=BMM_Additive
    rotationSpace=BCS_BoneSpace
    translationMode=BMM_Additive
    translationSpace=BCS_BoneSpace
    alphaBoolBlend {blendInTime=0.2, blendOption=Linear, blendOutTime=0.2}
    boneToModify {boneName=dyn_skirt_01_l}
    rotation=<rot 5.9284 0.0019 0.0011>
   }
   n37 = "Transform (Modify) Bone - Bone: dyn_skirt_01_front_l"(ComponentPose=@n36, bAlphaBoolEnabled=True) {
    alphaInputType=Bool
    rotationMode=BMM_Additive
    rotationSpace=BCS_BoneSpace
    translationMode=BMM_Additive
    translationSpace=BCS_BoneSpace
    alphaBoolBlend {blendInTime=0.2, blendOption=Linear, blendOutTime=0.2}
    boneToModify {boneName=dyn_skirt_01_front_l}
    rotation=<rot 6.1968 0.0238 -0.0128>
   }
   n38 = "Transform (Modify) Bone - Bone: dyn_skirt_01_l"(ComponentPose=@n37, bAlphaBoolEnabled=True) {
    alphaInputType=Bool
    rotationMode=BMM_Additive
    rotationSpace=BCS_BoneSpace
    translationMode=BMM_Additive
    translationSpace=BCS_BoneSpace
    alphaBoolBlend {blendInTime=0.2, blendOption=Linear, blendOutTime=0.2}
    boneToModify {boneName=dyn_skirt_01_l}
    rotation=<rot 3.8881 0.0028 -0.0016>
   }
   n39 = "Transform (Modify) Bone - Bone: dyn_skirt_01_front_r"(ComponentPose=@n38, bAlphaBoolEnabled=True) {
    alphaInputType=Bool
    rotationMode=BMM_Additive
    rotationSpace=BCS_BoneSpace
    translationMode=BMM_Additive
    translationSpace=BCS_BoneSpace
    alphaBoolBlend {blendInTime=0.2, blendOption=Linear, blendOutTime=0.2}
    boneToModify {boneName=dyn_skirt_01_front_r}
    rotation {pitch=5}
   }
   n40 = "Transform (Modify) Bone - Bone: dyn_skirt_01_r"(ComponentPose=@n39, bAlphaBoolEnabled=True) {
    alphaInputType=Bool
    rotationMode=BMM_Additive
    rotationSpace=BCS_BoneSpace
    translationMode=BMM_Additive
    translationSpace=BCS_BoneSpace
    alphaBoolBlend {blendInTime=0.2, blendOption=Linear, blendOutTime=0.2}
    boneToModify {boneName=dyn_skirt_01_r}
    rotation {pitch=5}
   }
   n41 = RigidBody!(ComponentPose=@n40, Alpha, OverrideWorldGravity:Vector="0,0,0", ExternalForce) -> (Pose) {
    bOverrideWorldGravity=true
    overridePhysicsAsset=&/M_MED_LatteStir_Vita_RightLeg_SkirtPhysics
    simulationSpace=BaseBoneSpace
    alphaBoolBlend {blendOption=Linear}
    baseBoneRef {boneName=thigh_r}
    componentAppliedLinearAccClamp=<3000 3000 3000>
    simSpaceSettings {
     maxAngularAcceleration=1
     maxAngularVelocity=3
     maxLinearAcceleration=700
     maxLinearVelocity=500
     worldAlpha=0.6
     externalLinearDragV=<0.2 0.2 0.2>
    }
   }
   n42 = RigidBody(ComponentPose=@n41, OverrideWorldGravity="0,0,-980") {
    bOverrideWorldGravity=true
    overridePhysicsAsset=&/M_MED_LatteStir_Vita_LeftLeg_SkirtPhysics
    simulationSpace=BaseBoneSpace
    alphaBoolBlend {blendOption=Linear}
    baseBoneRef {boneName=thigh_l}
    componentAppliedLinearAccClamp=<3000 3000 3000>
    simSpaceSettings {
     maxAngularAcceleration=1
     maxAngularVelocity=3
     maxLinearAcceleration=700
     maxLinearVelocity=500
     worldAlpha=0.6
     externalLinearDragV=<0.2 0.2 0.2>
    }
   }
   n43 = Anim Dynamics - Bone: dyn_skirt_03_bck!(ComponentPose=@n42, Alpha=1, GravityOverride:Vector="0,0,-980", ExternalForce="0,0,0") -> (Pose) {
    angularDampingOverride=0.9999
    angularSpringConstant=1000
    bAngularSpring=true
    bOverrideAngularDamping=true
    bUseGravityOverride=true
    gravityScale=0.35
    simulationSpace=RootRelative
    alphaBoolBlend {blendOption=Linear}
    boundBone {boneName=dyn_skirt_03_bck}
    componentAppliedLinearAccClamp=<9000 9000 9000>
    componentLinearAccScale=<0.5 0.5 0.5>
    componentLinearVelScale=<0.5 0.5 0.5>
    gravityOverride {z=-980}
    physicsBodyDefinitions [{
      boundBone {boneName=dyn_skirt_03_bck}
      boxExtents=<20 30 30>
      constraintSetup {angularLimitsMax=<0 1 10>, angularLimitsMin=<0 -20 -10>, angularTarget=<1 0 0>}
      localJointOffset=<3 0 0>
     }]
   }
   n44 = Anim Dynamics - Bone: dyn_skirt_03_bck_l!(ComponentPose=@n43, Alpha=1, GravityOverride:Vector="0,0,-980", ExternalForce="0,0,0") -> (Pose) {
    angularDampingOverride=0.9999
    angularSpringConstant=1000
    bAngularSpring=true
    bOverrideAngularDamping=true
    bUseGravityOverride=true
    gravityScale=0.35
    alphaBoolBlend {blendOption=Linear}
    boundBone {boneName=dyn_skirt_03_bck_l}
    componentAppliedLinearAccClamp=<9000 9000 9000>
    componentLinearAccScale=<0.5 0.5 0.5>
    componentLinearVelScale=<0.5 0.5 0.5>
    gravityOverride {z=-980}
    physicsBodyDefinitions [{
      boundBone {boneName=dyn_skirt_03_bck_l}
      boxExtents=<20 30 30>
      constraintSetup {angularLimitsMax=<0 1 0>, angularLimitsMin=<0 -20 -10>, angularTarget=<1 0 0>}
      localJointOffset=<3 0 0>
     }]
   }
   n45 = Anim Dynamics - Bone: dyn_skirt_03_bck_r!(ComponentPose=@n44, Alpha=1, GravityOverride:Vector="0,0,-980", ExternalForce="0,0,0") -> (Pose) {
    angularDampingOverride=0.9999
    angularSpringConstant=1000
    bAngularSpring=true
    bOverrideAngularDamping=true
    bUseGravityOverride=true
    gravityScale=0.35
    simulationSpace=RootRelative
    alphaBoolBlend {blendOption=Linear}
    boundBone {boneName=dyn_skirt_03_bck_r}
    componentAppliedLinearAccClamp=<9000 9000 9000>
    componentLinearAccScale=<0.5 0.5 0.5>
    componentLinearVelScale=<0.5 0.5 0.5>
    gravityOverride {z=-980}
    physicsBodyDefinitions [{
      boundBone {boneName=dyn_skirt_03_bck_r}
      boxExtents=<20 30 30>
      constraintSetup {angularLimitsMax=<0 1 5>, angularLimitsMin=<0 -20 -10>, angularTarget=<1 0 0>}
      localJointOffset=<3 0 0>
     }]
   }
   n46 = Anim Dynamics - Bone: dyn_skirt_04_bck!(ComponentPose=@n45, Alpha=1, GravityOverride:Vector="0,0,-980", ExternalForce="0,0,0") -> (Pose) {
    angularDampingOverride=0.9999
    angularSpringConstant=1000
    bAngularSpring=true
    bOverrideAngularDamping=true
    bUseGravityOverride=true
    gravityScale=0.35
    simulationSpace=RootRelative
    alphaBoolBlend {blendOption=Linear}
    boundBone {boneName=dyn_skirt_04_bck}
    componentAppliedLinearAccClamp=<9000 9000 9000>
    componentLinearAccScale=<0.5 0.5 0.5>
    componentLinearVelScale=<0.5 0.5 0.5>
    gravityOverride {z=-980}
    physicsBodyDefinitions [{
      boundBone {boneName=dyn_skirt_04_bck}
      boxExtents=<20 30 30>
      constraintSetup {angularLimitsMax=<0 1 5>, angularLimitsMin=<0 -20 -10>, angularTarget=<1 0 0>}
      localJointOffset=<3 0 0>
     }]
   }
   n47 = Anim Dynamics - Bone: dyn_skirt_04_bck_l!(ComponentPose=@n46, Alpha=1, GravityOverride:Vector="0,0,-980", ExternalForce="0,0,0") -> (Pose) {
    angularDampingOverride=0.9999
    angularSpringConstant=1000
    bAngularSpring=true
    bOverrideAngularDamping=true
    bUseGravityOverride=true
    gravityScale=0.35
    alphaBoolBlend {blendOption=Linear}
    boundBone {boneName=dyn_skirt_04_bck_l}
    componentAppliedLinearAccClamp=<9000 9000 9000>
    componentLinearAccScale=<0.5 0.5 0.5>
    componentLinearVelScale=<0.5 0.5 0.5>
    gravityOverride {z=-980}
    physicsBodyDefinitions [{
      boundBone {boneName=dyn_skirt_04_bck_l}
      boxExtents=<20 30 30>
      constraintSetup {angularLimitsMax=<0 1 0>, angularLimitsMin=<0 -20 -10>, angularTarget=<1 0 0>}
      localJointOffset=<3 0 0>
     }]
   }
   n48 = Anim Dynamics - Bone: dyn_skirt_04_bck_r!(ComponentPose=@n47, Alpha=1, GravityOverride:Vector="0,0,-980", ExternalForce="0,0,0") -> (Pose) {
    angularDampingOverride=0.9999
    angularSpringConstant=1000
    bAngularSpring=true
    bOverrideAngularDamping=true
    bUseGravityOverride=true
    gravityScale=0.35
    simulationSpace=RootRelative
    alphaBoolBlend {blendOption=Linear}
    boundBone {boneName=dyn_skirt_04_bck_r}
    componentAppliedLinearAccClamp=<9000 9000 9000>
    componentLinearAccScale=<0.5 0.5 0.5>
    componentLinearVelScale=<0.5 0.5 0.5>
    gravityOverride {z=-980}
    physicsBodyDefinitions [{
      boundBone {boneName=dyn_skirt_04_bck_r}
      boxExtents=<20 30 30>
      constraintSetup {angularLimitsMax=<0 1 5>, angularLimitsMin=<0 -20 -10>, angularTarget=<1 0 0>}
      localJointOffset=<3 0 0>
     }]
   }
   n49 = Anim Dynamics - Bone: dyn_skirt_05_bck!(ComponentPose=@n48, Alpha=1, GravityOverride:Vector="0,0,-980", ExternalForce="0,0,0") -> (Pose) {
    angularDampingOverride=0.9999
    angularSpringConstant=1000
    bAngularSpring=true
    bOverrideAngularDamping=true
    bUseGravityOverride=true
    gravityScale=0.35
    simulationSpace=RootRelative
    alphaBoolBlend {blendOption=Linear}
    boundBone {boneName=dyn_skirt_05_bck}
    componentAppliedLinearAccClamp=<9000 9000 9000>
    componentLinearAccScale=<0.5 0.5 0.5>
    componentLinearVelScale=<0.5 0.5 0.5>
    gravityOverride {z=-980}
    physicsBodyDefinitions [{
      boundBone {boneName=dyn_skirt_05_bck}
      boxExtents=<20 30 30>
      constraintSetup {angularLimitsMax=<0 5 10>, angularLimitsMin=<0 -20 -10>, angularTarget=<1 0 0>}
      localJointOffset=<3 0 0>
     }]
   }
   n50 = Anim Dynamics - Bone: dyn_skirt_05_bck_l!(ComponentPose=@n49, Alpha=1, GravityOverride:Vector="0,0,-980", ExternalForce="0,0,0") -> (Pose) {
    angularDampingOverride=0.9999
    angularSpringConstant=1000
    bAngularSpring=true
    bOverrideAngularDamping=true
    bUseGravityOverride=true
    gravityScale=0.35
    alphaBoolBlend {blendOption=Linear}
    boundBone {boneName=dyn_skirt_05_bck_l}
    componentAppliedLinearAccClamp=<9000 9000 9000>
    componentLinearAccScale=<0.5 0.5 0.5>
    componentLinearVelScale=<0.5 0.5 0.5>
    gravityOverride {z=-980}
    physicsBodyDefinitions [{
      boundBone {boneName=dyn_skirt_05_bck_l}
      boxExtents=<20 30 30>
      constraintSetup {angularLimitsMax=<0 5 0>, angularLimitsMin=<0 -20 -10>, angularTarget=<1 0 0>}
      localJointOffset=<3 0 0>
     }]
   }
   n51 = Anim Dynamics - Bone: dyn_skirt_05_bck_r!(ComponentPose=@n50, Alpha=1, GravityOverride:Vector="0,0,-980", ExternalForce="0,0,0") -> (Pose) {
    angularDampingOverride=0.9999
    angularSpringConstant=1000
    bAngularSpring=true
    bOverrideAngularDamping=true
    bUseGravityOverride=true
    gravityScale=0.35
    simulationSpace=RootRelative
    alphaBoolBlend {blendOption=Linear}
    boundBone {boneName=dyn_skirt_05_bck_r}
    componentAppliedLinearAccClamp=<9000 9000 9000>
    componentLinearAccScale=<0.5 0.5 0.5>
    componentLinearVelScale=<0.5 0.5 0.5>
    gravityOverride {z=-980}
    physicsBodyDefinitions [{
      boundBone {boneName=dyn_skirt_05_bck_r}
      boxExtents=<20 30 30>
      constraintSetup {angularLimitsMax=<0 5 5>, angularLimitsMin=<0 -20 -10>, angularTarget=<1 0 0>}
      localJointOffset=<3 0 0>
     }]
   }
   n52 = "Constraint - dyn_skirt_03_fr (RR)"!(ComponentPose=@n51, Alpha=1, ConstraintWeights_0=0.5, ConstraintWeights_1:float=0.5) -> (Pose) {
    alphaBoolBlend {blendOption=Linear}
    boneToModify {boneName=dyn_skirt_03_fr}
    constraintSetup [{transformType=Rotation, targetBone {boneName=dyn_skirt_03_front_l}}, {transformType=Rotation, targetBone {boneName=dyn_skirt_03_front_r}}]
    constraintWeights [1,1]
   }
   n53 = "Constraint - dyn_skirt_04_fr (RR)"!(ComponentPose=@n52, Alpha=1, ConstraintWeights_0=0.5, ConstraintWeights_1:float=0.5) -> (Pose) {
    alphaBoolBlend {blendOption=Linear}
    boneToModify {boneName=dyn_skirt_04_fr}
    constraintSetup [{transformType=Rotation, targetBone {boneName=dyn_skirt_04_front_l}}, {transformType=Rotation, targetBone {boneName=dyn_skirt_04_front_r}}]
    constraintWeights [1,1]
   }
   n54 = "Constraint - dyn_skirt_05_fr (RR)"!(ComponentPose=@n53, Alpha=1, ConstraintWeights_0=0.5, ConstraintWeights_1:float=0.5) -> (Pose) {
    alphaBoolBlend {blendOption=Linear}
    boneToModify {boneName=dyn_skirt_05_fr}
    constraintSetup [{transformType=Rotation, targetBone {boneName=dyn_skirt_05_front_l}}, {transformType=Rotation, targetBone {boneName=dyn_skirt_05_front_r}}]
    constraintWeights [1,1]
   }
   n55 = "Transform (Modify) Bone - Bone: dyn_skirt_03_bck_l"!(ComponentPose=@n54, bAlphaBoolEnabled:bool=True) -> (Pose) {
    alphaInputType=Bool
    rotationMode=BMM_Additive
    rotationSpace=BCS_BoneSpace
    translationMode=BMM_Additive
    translationSpace=BCS_BoneSpace
    alphaBoolBlend {blendInTime=0.2, blendOption=Linear, blendOutTime=0.2}
    boneToModify {boneName=dyn_skirt_03_bck_l}
    rotation=<rot 14.2552 0.002 0.0021>
    translation=<-0.4431 -0.2249 1.7409>
   }
   n56 = "Transform (Modify) Bone - Bone: dyn_skirt_03_bck"!(ComponentPose=@n55, bAlphaBoolEnabled:bool=True) -> (Pose) {
    alphaInputType=Bool
    rotationMode=BMM_Additive
    rotationSpace=BCS_BoneSpace
    translationMode=BMM_Additive
    translationSpace=BCS_BoneSpace
    alphaBoolBlend {blendInTime=0.2, blendOption=Linear, blendOutTime=0.2}
    boneToModify {boneName=dyn_skirt_03_bck}
    rotation=<rot 18.4749 0.025 0.0123>
    translation=<-2.4629 0.8213 0.2305>
   }
   n57 = "Transform (Modify) Bone - Bone: dyn_skirt_03_bck_r"!(ComponentPose=@n56, bAlphaBoolEnabled:bool=True) -> (Pose) {
    alphaInputType=Bool
    rotationMode=BMM_Additive
    rotationSpace=BCS_BoneSpace
    translationMode=BMM_Additive
    translationSpace=BCS_BoneSpace
    alphaBoolBlend {blendInTime=0.2, blendOption=Linear, blendOutTime=0.2}
    boneToModify {boneName=dyn_skirt_03_bck_r}
    rotation=<rot 13.9072 -0.0053 -0.004>
   }
   n58 = "Transform (Modify) Bone - Bone: dyn_skirt_01_bck_l"!(ComponentPose=@n57, bAlphaBoolEnabled=True) -> (Pose) {
    alphaInputType=Bool
    rotationMode=BMM_Additive
    rotationSpace=BCS_BoneSpace
    translationMode=BMM_Additive
    translationSpace=BCS_BoneSpace
    alphaBoolBlend {blendInTime=0.2, blendOption=Linear, blendOutTime=0.2}
    boneToModify {boneName=dyn_skirt_01_bck_l}
    translation=<-0.4431 -0.2249 1.7409>
   }
   n59 = "Transform (Modify) Bone - Bone: dyn_skirt_01_bck"!(ComponentPose=@n58, bAlphaBoolEnabled:bool=True) -> (Pose) {
    alphaInputType=Bool
    rotationMode=BMM_Additive
    rotationSpace=BCS_BoneSpace
    translationMode=BMM_Additive
    translationSpace=BCS_BoneSpace
    alphaBoolBlend {blendInTime=0.2, blendOption=Linear, blendOutTime=0.2}
    boneToModify {boneName=dyn_skirt_01_bck}
    rotation=<rot 0.0842 -0.3836 -0.0199>
    translation=<-2.4629 0.8213 0.2305>
   }
   n60 = "Transform (Modify) Bone - Bone: dyn_skirt_01_bck_r"!(ComponentPose=@n59, bAlphaBoolEnabled:bool=True) -> (Pose) {
    alphaInputType=Bool
    rotationMode=BMM_Additive
    rotationSpace=BCS_BoneSpace
    translationMode=BMM_Additive
    translationSpace=BCS_BoneSpace
    alphaBoolBlend {blendInTime=0.2, blendOption=Linear, blendOutTime=0.2}
    boneToModify {boneName=dyn_skirt_01_bck_r}
    rotation=<rot 0.0152 6.9709 -0.0137>
   }
   n61 = "Transform (Modify) Bone - Bone: dyn_skirt_01_bck_r"(ComponentPose=@n60, bAlphaBoolEnabled=True) {
    alphaInputType=Bool
    rotationMode=BMM_Additive
    rotationSpace=BCS_BoneSpace
    translationMode=BMM_Additive
    translationSpace=BCS_BoneSpace
    alphaBoolBlend {blendInTime=0.2, blendOption=Linear, blendOutTime=0.2}
    boneToModify {boneName=dyn_skirt_01_bck_r}
    rotation=<rot 0.0185 15.6651 -0.0044>
   }
   n62 = "Transform (Modify) Bone - Bone: dyn_skirt_02_bck_l"!(ComponentPose=@n61, bAlphaBoolEnabled=True) -> (Pose) {
    alphaInputType=Bool
    rotationMode=BMM_Additive
    rotationSpace=BCS_BoneSpace
    translationMode=BMM_Additive
    translationSpace=BCS_BoneSpace
    alphaBoolBlend {blendInTime=0.2, blendOption=Linear, blendOutTime=0.2}
    boneToModify {boneName=dyn_skirt_02_bck_l}
    rotation=<rot -10.3027 -0.038 -0.002>
    translation=<-0.4431 -0.2249>
   }
   n63 = "Transform (Modify) Bone - Bone: dyn_skirt_02_l"!(ComponentPose=@n62, bAlphaBoolEnabled=True) -> (Pose) {
    alphaInputType=Bool
    rotationMode=BMM_Additive
    rotationSpace=BCS_BoneSpace
    translationMode=BMM_Additive
    translationSpace=BCS_BoneSpace
    alphaBoolBlend {blendInTime=0.2, blendOption=Linear, blendOutTime=0.2}
    boneToModify {boneName=dyn_skirt_02_l}
    rotation=<rot -0.0173 8.7607 -0.0089>
    translation=<-0.4431 -0.2249>
   }
   n64 = "Transform (Modify) Bone - Bone: dyn_skirt_02_bck"!(ComponentPose=@n63, bAlphaBoolEnabled:bool=True) -> (Pose) {
    alphaInputType=Bool
    rotationMode=BMM_Additive
    rotationSpace=BCS_BoneSpace
    translationMode=BMM_Additive
    translationSpace=BCS_BoneSpace
    alphaBoolBlend {blendInTime=0.2, blendOption=Linear, blendOutTime=0.2}
    boneToModify {boneName=dyn_skirt_02_bck}
    rotation=<rot -11.8817 0.0269 -0.0146>
    translation=<-0.4431 -0.2249>
   }
   n65 = "Transform (Modify) Bone - Bone: dyn_skirt_02_fr"!(ComponentPose=@n64, bAlphaBoolEnabled:bool=True) -> (Pose) {
    alphaInputType=Bool
    rotationMode=BMM_Additive
    rotationSpace=BCS_BoneSpace
    translationMode=BMM_Additive
    translationSpace=BCS_BoneSpace
    alphaBoolBlend {blendInTime=0.2, blendOption=Linear, blendOutTime=0.2}
    boneToModify {boneName=dyn_skirt_02_fr}
    rotation {pitch=20}
   }
   n66 = "Transform (Modify) Bone - Bone: dyn_skirt_04_fr"!(ComponentPose=@n65, bAlphaBoolEnabled:bool=True) -> (Pose) {
    alphaInputType=Bool
    rotationMode=BMM_Additive
    rotationSpace=BCS_BoneSpace
    translationMode=BMM_Additive
    translationSpace=BCS_BoneSpace
    alphaBoolBlend {blendInTime=0.2, blendOption=Linear, blendOutTime=0.2}
    boneToModify {boneName=dyn_skirt_04_fr}
    rotation {pitch=10}
   }
   Save cached pose 'CRCache_baseSkirt_setup'(Pose=Component To Local(ComponentPose=@n66)) {cacheName=CRCache_baseSkirt_setup}
   n67 = Blend Poses by bool(
    BlendPose_0=Use cached pose 'CRCache_KneeSliding_setup'()
    BlendPose_1=Use cached pose 'CRCache_baseSkirt_setup'()
    BlendTime_0=0.2
    bActiveValue=Get bIsKneeSliding()
   ) {blendType=Linear}
   n68 = Use cached pose 'CRCache_BaseballSliding_Setup'()
   n69 = Get bIsBaseballSliding()
   n70 = "Transform (Modify) Bone - Bone: dyn_skirt_01_bck_l"(
    ComponentPose=Local To Component(LocalPose=Blend Poses by bool(BlendPose_0=@n68, BlendPose_1=@n67, BlendTime_0=0.5, BlendTime_1=0.2, bActiveValue=@n69) {blendType=Linear})
    bAlphaBoolEnabled=True
   ) {
    alphaInputType=Bool
    rotationMode=BMM_Additive
    rotationSpace=BCS_BoneSpace
    translationMode=BMM_Additive
    translationSpace=BCS_BoneSpace
    alphaBoolBlend {blendOption=Linear}
    boneToModify {boneName=dyn_skirt_01_bck_l}
    rotation {pitch=6}
   }
   n71 = "Transform (Modify) Bone - Bone: dyn_skirt_01_bck_r"(ComponentPose=@n70, bAlphaBoolEnabled=True) {
    alphaInputType=Bool
    rotationMode=BMM_Additive
    rotationSpace=BCS_BoneSpace
    translationMode=BMM_Additive
    translationSpace=BCS_BoneSpace
    alphaBoolBlend {blendOption=Linear}
    boneToModify {boneName=dyn_skirt_01_bck_r}
    rotation {pitch=6}
   }
   n72 = "Transform (Modify) Bone - Bone: dyn_skirt_01_bck"(ComponentPose=@n71, bAlphaBoolEnabled=True) {
    alphaInputType=Bool
    rotationMode=BMM_Additive
    rotationSpace=BCS_BoneSpace
    translationMode=BMM_Additive
    translationSpace=BCS_BoneSpace
    alphaBoolBlend {blendOption=Linear}
    boneToModify {boneName=dyn_skirt_01_bck}
    rotation {pitch=6}
   }
   Save cached pose 'CRCache_Skirt_Dynamics'(Pose=Component To Local(ComponentPose=@n72)) {cacheName=CRCache_Skirt_Dynamics}
   n73 = "Transform (Modify) Bone - Bone: dyn_skirt_01_fr"!(ComponentPose=Local To Component(LocalPose=Use cached pose 'CRCache_InputPose'()), Alpha=1) -> (Pose) {
    rotationMode=BMM_Additive
    rotationSpace=BCS_BoneSpace
    alphaBoolBlend {blendOption=Linear}
    boneToModify {boneName=dyn_skirt_01_fr}
    rotation=<rot 26.773 8.281 -20.375>
   }
   n74 = "Transform (Modify) Bone - Bone: dyn_skirt_01_bck_l"!(ComponentPose=@n73, Alpha:float=1) -> (Pose) {
    rotationMode=BMM_Additive
    rotationSpace=BCS_BoneSpace
    translationMode=BMM_Additive
    translationSpace=BCS_BoneSpace
    alphaBoolBlend {blendOption=Linear}
    boneToModify {boneName=dyn_skirt_01_bck_l}
    rotation=<rot 34.5113 -0.7431 49.6123>
    translation=<-8.5313 4.299 -0.8113>
   }
   n75 = "Transform (Modify) Bone - Bone: dyn_skirt_02_bck_r"!(ComponentPose=@n74, Alpha=1) -> (Pose) {
    rotationMode=BMM_Additive
    rotationSpace=BCS_BoneSpace
    translationMode=BMM_Additive
    translationSpace=BCS_BoneSpace
    alphaBoolBlend {blendOption=Linear}
    boneToModify {boneName=dyn_skirt_02_bck_r}
    rotation=<rot -0.6814 -13.0356 -30.5215>
    translation=<1.4791 -5.3077 11.146>
   }
   n76 = "Transform (Modify) Bone - Bone: dyn_skirt_01_bck"!(ComponentPose=@n75, Alpha:float=1) -> (Pose) {
    rotationMode=BMM_Additive
    rotationSpace=BCS_BoneSpace
    translationMode=BMM_Additive
    translationSpace=BCS_BoneSpace
    alphaBoolBlend {blendOption=Linear}
    boneToModify {boneName=dyn_skirt_01_bck}
    rotation=<rot 14.8901 0 0>
    translation=<-10.655 0.0006 1.434>
   }
   n77 = "Transform (Modify) Bone - Bone: dyn_skirt_01_l"!(ComponentPose=@n76, Alpha:float=1) -> (Pose) {
    rotationMode=BMM_Additive
    rotationSpace=BCS_BoneSpace
    alphaBoolBlend {blendOption=Linear}
    boneToModify {boneName=dyn_skirt_01_l}
    rotation=<rot 41.4529 -1.5658 23.7392>
   }
   n78 = "Transform (Modify) Bone - Bone: dyn_skirt_01_r"!(ComponentPose=@n77, Alpha:float=1) -> (Pose) {
    rotationMode=BMM_Additive
    rotationSpace=BCS_BoneSpace
    alphaBoolBlend {blendOption=Linear}
    boneToModify {boneName=dyn_skirt_01_r}
    rotation=<rot 32.54 -12.717 -40.151>
   }
   n79 = RigidBody(ComponentPose=@n78, OverrideWorldGravity="0,-10000,0") {
    bOverrideWorldGravity=true
    overridePhysicsAsset=&/M_MED_LatteStir_Vita_SkirtKneeSlidingPhysics
    simulationSpace=BaseBoneSpace
    alphaBoolBlend {blendOption=Linear}
    baseBoneRef {boneName=pelvis}
    componentAppliedLinearAccClamp=<3000 3000 3000>
    componentLinearAccScale=<0.3 0.3 0.3>
    componentLinearVelScale=<0.3 0.3 0.3>
    simSpaceSettings {maxAngularAcceleration=1000, maxAngularVelocity=3, maxLinearAcceleration=700, maxLinearVelocity=500, velocityScaleZ=0}
   }
   Save cached pose 'CRCache_KneeSliding_setup'(Pose=Component To Local(ComponentPose=@n79)) {cacheName=CRCache_KneeSliding_setup}
   n80 = "Transform (Modify) Bone - Bone: dyn_skirt_01_fr"(ComponentPose=Local To Component(LocalPose=Use cached pose 'CRCache_InputPose'()), Alpha=1) {
    rotationMode=BMM_Additive
    rotationSpace=BCS_BoneSpace
    alphaBoolBlend {blendOption=Linear}
    boneToModify {boneName=dyn_skirt_01_fr}
    rotation=<rot -44.86 5.828 -49.8102>
   }
   n81 = "Transform (Modify) Bone - Bone: dyn_skirt_01_bck_l"(ComponentPose=@n80, Alpha=1) {
    rotationMode=BMM_Additive
    rotationSpace=BCS_BoneSpace
    translationMode=BMM_Additive
    translationSpace=BCS_BoneSpace
    alphaBoolBlend {blendOption=Linear}
    boneToModify {boneName=dyn_skirt_01_bck_l}
    rotation=<rot -50.8229 0 0>
    translation {x=4.7255, z=3.8509}
   }
   n82 = "Transform (Modify) Bone - Bone: dyn_skirt_01_bck_r"!(ComponentPose=@n81, Alpha:float=1) -> (Pose) {
    rotationMode=BMM_Additive
    rotationSpace=BCS_BoneSpace
    translationMode=BMM_Additive
    translationSpace=BCS_BoneSpace
    alphaBoolBlend {blendOption=Linear}
    boneToModify {boneName=dyn_skirt_01_bck_r}
    rotation=<rot -61.2104 0 0>
    translation=<7.3948 0 4.0636>
   }
   n83 = "Transform (Modify) Bone - Bone: dyn_skirt_01_bck"(ComponentPose=@n82, Alpha=1) {
    rotationMode=BMM_Additive
    rotationSpace=BCS_BoneSpace
    translationMode=BMM_Additive
    translationSpace=BCS_BoneSpace
    alphaBoolBlend {blendOption=Linear}
    boneToModify {boneName=dyn_skirt_01_bck}
    rotation=<rot -63.695 0.0001 -0.0001>
    translation=<12.8683 0 6.3613>
   }
   n84 = "Transform (Modify) Bone - Bone: dyn_skirt_01_l"(ComponentPose=@n83, Alpha=1) {
    rotationMode=BMM_Additive
    rotationSpace=BCS_BoneSpace
    alphaBoolBlend {blendOption=Linear}
    boneToModify {boneName=dyn_skirt_01_l}
    rotation=<rot -9.3355 8.7195 -1.4252>
   }
   n85 = "Transform (Modify) Bone - Bone: dyn_skirt_01_r"(ComponentPose=@n84, Alpha=1) {
    rotationMode=BMM_Additive
    rotationSpace=BCS_BoneSpace
    alphaBoolBlend {blendOption=Linear}
    boneToModify {boneName=dyn_skirt_01_r}
    rotation=<rot -12.0593 -31.6727 -23.8143>
   }
   n86 = "Transform (Modify) Bone - Bone: base_skirt_01_front_r"!(ComponentPose=@n85, Alpha=1) -> (Pose) {
    rotationMode=BMM_Additive
    rotationSpace=BCS_BoneSpace
    translationMode=BMM_Additive
    translationSpace=BCS_BoneSpace
    alphaBoolBlend {blendOption=Linear}
    boneToModify {boneName=base_skirt_01_front_r}
    rotation=<rot -3.2675 7.3743 -37.5186>
   }
   n87 = "Transform (Modify) Bone - Bone: dyn_skirt_01_front_l"!(ComponentPose=@n86, Alpha:float=1) -> (Pose) {
    rotationMode=BMM_Additive
    rotationSpace=BCS_BoneSpace
    translationMode=BMM_Additive
    translationSpace=BCS_BoneSpace
    alphaBoolBlend {blendOption=Linear}
    boneToModify {boneName=dyn_skirt_01_front_l}
    rotation=<rot -6.2938 -0.0001 -22.6339>
   }
   n88 = RigidBody(ComponentPose=@n87, OverrideWorldGravity="0,2000,0") {
    bOverrideWorldGravity=true
    overridePhysicsAsset=&/M_MED_LatteSitr_Vita_SkirtSliding_Physics
    simulationSpace=BaseBoneSpace
    alphaBoolBlend {blendOption=Linear}
    baseBoneRef {boneName=pelvis}
    componentAppliedLinearAccClamp=<3000 3000 3000>
    componentLinearAccScale=<0.3 0.3 0.3>
    componentLinearVelScale=<0.3 0.3 0.3>
    simSpaceSettings {maxAngularAcceleration=1000, maxAngularVelocity=3, maxLinearAcceleration=700, maxLinearVelocity=500, velocityScaleZ=0}
   }
   Save cached pose 'CRCache_BaseballSliding_Setup'(Pose=Component To Local(ComponentPose=@n88)) {cacheName=CRCache_BaseballSliding_Setup}
  ]}
 EventGraph {nodes [
   n1 = Event BlueprintUpdateAnimation()
   > n2 = Sequence!(execute) -> (then_0, then_1, then_2:Exec, then_3:Exec, then_4:Exec, then_5:Exec)
   n3 = Get bIsMoving2D()
   Set Moving2d(
    execute=@n2.then_1
    Moving2d:double=FInterpTo(Current=Get Moving2d(), Target="To Float (Boolean)"(InBool=@n3), DeltaTime=@n1.DeltaTimeX, InterpSpeed=2.5)
   ) -> (Output_Get:double)
   n4 = Get bIsSkydiveFloating()
   > Set isFloating(isFloating:double=FInterpTo(Current=Get isFloating(), Target="To Float (Boolean)"(InBool=@n4), DeltaTime=@n1.DeltaTimeX, InterpSpeed=3)) -> (Output_Get:double)
   n5 = Get bIsMoving2D()
   > Set CrouchIdle(CrouchIdle:bool=AND Boolean(A=Get bIsCrouching(), B=NOT Boolean(A=@n5))) -> (Output_Get:bool)
   n6 = Get bIsMoving2D()
   n7 = Get bIsCrouching()
   > Set Idle(Idle:bool=AND Boolean(A=NOT Boolean(A=@n6), B=NOT Boolean(A=@n7))) -> (Output_Get:bool)
   > n8 = Branch(Condition=Get CrouchIdle() -> (CrouchIdle:bool))
   > Set GravityOverride RigidBodies(GravityOverride RigidBodies:Vector="0,0,-980") -> (Output_Get:Vector)
   Set GravityOverride RigidBodies(execute=@n8.else, GravityOverride RigidBodies=Get GravityOverride())
   Reroute Node(InputPin=@n2.then_3)
   n9 = PlayerMovement(execute=@n2.then_4, AnimInstance=Self-Reference(), XDotProductCoefficient=100, YDotProductCoefficient=100, ZDotProductCoefficient=100)
   n10 = Get bIsCrouching()
   n11 = Get bIsSkyDiving()
   n12 = BreakVector(InVec=MakeVector(X=@n9, Y=@n9.YMovement))
   > Set forwardBool(forwardBool:bool=AND Boolean!(A="float > float"(A=@n12.Y, B=40), B=NOT Boolean(A=@n10), C:bool=NOT Boolean(A=@n11)) -> (ReturnValue)) -> (Output_Get:bool)
   n13 = Set kneerotation(
    execute=@n2.then_5
    kneerotation:double=GetAlphaFromReferenceTransformDelta(SkeletalMeshComponent=Get Skel_Mesh(), BoneName=calf_l, BaseName=thigh_l, InputRange=(X=80,Y=140), TwistAxis="0,0,1")
   ) -> (Output_Get:double)
   > n14 = Branch(Condition="float >= float"(A=@n13, B=0.2))
   > Set LatteStirEmote(LatteStirEmote:bool=Get LatteStirEmote() -> (LatteStirEmote:bool)) -> (Output_Get:bool)
   Set LatteStirEmote(execute=@n14.else, LatteStirEmote=false)
   n15 = Cast To FortPlayerAnimInstance(execute=@n2.then_2, Object=GetAnimInstance(self=GetOwnerSkeletalMeshComponent()))
   > Set goingCommando(goingCommando:bool=false) -> (Output_Get:bool)
   > n16 = Switch on EFortWeaponCoreAnimation!(execute, Selection:EFortWeaponCoreAnimation Enum=Get WeaponCoreAnim(self=@n15)) -> (
    Melee:Exec
    Pistol:Exec
    Shotgun:Exec
    PaperBlueprint:Exec
    Rifle:Exec
    MeleeOneHand:Exec
    MachinePistol:Exec
    RocketLauncher:Exec
    GrenadeLauncher:Exec
    GoingCommando:Exec
    AssaultRifle:Exec
    TacticalShotgun:Exec
    SniperRifle:Exec
    TrapPlacement:Exec
    ShoulderLauncher:Exec
    AbilityDecoTool:Exec
    Crossbow:Exec
    C4:Exec
    RemoteControl:Exec
    DualWield:Exec
    AR_BullPup:Exec
    AR_ForwardGrip:Exec
    MedPackPaddles:Exec
    SMG_P90:Exec
    AR_DrumGun:Exec
    Consumable_Small:Exec
    Consumable_Large:Exec
    Balloon:Exec
    MountedTurret:Exec
    CreativeTool:Exec
    ExplosiveBow:Exec
    AshtonIndigo:Exec
    AshtonChicago:Exec
    MeleeDualWield:Exec
    Unarmed:Exec
   )
   Branch(execute=@n16.GoingCommando)
   > Set goingCommando(goingCommando=1)
   > Branch(Condition=Get bIsSkyDiving())
   > Set goingCommando(goingCommando=false)
   n17 = "Gravity Override State Function (Parameter Asset) (Nativized)"(
    execute=@n2.then_0
    CharacterPartAnimInstance=Self-Reference()
    GameTimeInSeconds=GetGameTimeInSeconds()
    Parameters=&/M_MED_LatteStir_Vita_Parameters
   )
   > Set GravityOverride(GravityOverride=@n17)
   > Set RBAN Settings Out(RBAN Settings Out=@n17.OutRigidBodyAnimNodeParameters)
   n18 = GetAnimInstance(self=GetOwnerSkeletalMeshComponent())
   > Is Valid(InputObject=@n18)
   > Set celebration(celebration:bool=Montage_IsPlaying(self=@n18, Montage=/BRCosmetics/Animation/Game/MainPlayer/Emotes/Emote_Celebration_Female_M)) -> (Output_Get:bool)
   > Set LatteStirEmote(LatteStirEmote=Montage_IsPlaying(self=@n18, Montage=/BRCosmetics/Animation/Game/MainPlayer/Emotes/LatteStir/Hero/Emote_LatteStir_Hero_M))
   > Set isCrouch(isCrouch:double="To Float (Boolean)"(InBool=Get bIsCrouching())) -> (Output_Get:double)
   n19 = Event BlueprintProcessAnimInputEvent()
   > n20 = Branch(Condition=IsNamed(self=@n19.AnimInputEvent, InEventName=aie.Player.Zipline.Entered))
   > Set OnZipline(OnZipline:bool=1) -> (Output_Get:bool)
   Branch(execute=@n20.else, Condition=IsNamed(self=@n19.AnimInputEvent, InEventName=aie.Player.Zipline.Exited))
   > Set OnZipline(OnZipline=false)
   Event BlueprintInitializeAnimation()
   > Set Skel_Mesh(Skel_Mesh=GetOwnerSkeletalMeshComponent())
   Event OnParachuteOpened()
   > Set openParachute(openParachute:double=1) -> (Output_Get:double)
   > Delay(Duration=2.5)
   > Set openParachute(openParachute=0)
  ]}
]
variables [
 ACC/VEL Zipline {type=Vector}
 celebration {type=boolean}
 CrouchIdle {type=boolean}
 forwardBool {type=boolean}
 goingCommando {type=boolean}
 GravityOverride {type=Vector}
 GravityOverride RigidBodies {type=Vector}
 Idle {type=boolean}
 isCrouch {type=number}
 isFloating {type=number}
 kneerotation {type=number}
 LatteStirEmote {type=boolean}
 Moving2d {type=number}
 OnZipline {type=boolean}
 openParachute {type=number}
 RBAN Settings Out {type=FortRigidBodyAnimNodeParameters}
 Skel_Mesh {type=SkeletalMeshComponent}
]
