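# Physics asset for the face accessory mesh of the male LatteStir Vita cosmetic body, modeling three dangling hair-bone chains (left, center, right).
# Each chain is a run of capsules rooted to spine_05 by a free-rotation constraint.
# Chain joints allow only one swing axis, with a tighter limit toward the tips.
# Pelvis, spine and clavicle capsules supply torso shapes, with free constraints linking spine and clavicles while the pelvis stays unconstrained.
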PhysicsAsset /BRCosmetics/Characters/Player/Male/Medium/Bodies/M_MED_LatteStir_Vita/Meshes/Parts/M_MED_LatteStir_Vita_FaceAcc_Physics
bodies {
 clavicle_l [{
   shapeType=Capsule
   length=8.6429
   radius=6.8195
   shapeName=clavicle_l_capsule
   center=<2.9414 -0.1721 -2.5725>
   rotation=<rot -0.0001 -83.1598 -90>
  }]
 clavicle_r [{
   shapeType=Capsule
   length=8.6429
   radius=6.8195
   shapeName=clavicle_r_capsule
   center=<-2.9414 0.1721 2.5725>
   rotation=<rot -0.0001 -83.1598 90>
  }]
 dyn_hair_2_c [{shapeType=Capsule, length=1, radius=1}]
 dyn_hair_2_l [{shapeType=Capsule, length=1, radius=1}]
 dyn_hair_2_r [{shapeType=Capsule, length=1, radius=1}]
 dyn_hair_3_c [{shapeType=Capsule, length=3.4995, radius=3.6749, shapeName=dyn_C_hair_1_capsule, center=<5.4739 0 0.6866>, rotation=<rot 85.5986 0 0>}]
 dyn_hair_3_l [{
   shapeType=Capsule
   length=3.7831
   radius=2.7118
   shapeName=dyn_C_hair_1_capsule
   center=<3.9637 0.6928 -0.1129>
   rotation=<rot 81.2114 89.9995 179.9995>
  }]
 dyn_hair_3_r [
  {
   shapeType=Capsule
   length=3.2407
   radius=3.2542
   shapeName=dyn_C_hair_1_capsule
   center=<3.9638 -0.9125 0.6371>
   rotation=<rot -82.4517 90.0646 179.9349>
  }
  {
   shapeType=Capsule
   length=6.1692
   radius=2.4005
   shapeName=dyn_hair_3_r_capsule
   center=<5.9959 -5.5208 -1.6492>
   rotation=<rot -56.69 116.2578 -172.5911>
  }
 ]
 dyn_hair_4_c [{shapeType=Capsule, length=3.6191, radius=2.9054, shapeName=dyn_C_hair_1_capsule, center=<4.8324 0 0.1482>, rotation=<rot 88.2826 0 0>}]
 dyn_hair_4_l [{shapeType=Capsule, length=3.4698, radius=3.0548, shapeName=dyn_C_hair_1_capsule, center=<4.8048 1.7748 -0.1344>, rotation=<rot 88.2826 0 0>}]
 dyn_hair_4_r [{shapeType=Capsule, length=3.4698, radius=3.0548, shapeName=dyn_C_hair_1_capsule, center=<4.8436 -1.7748 0.2623>, rotation=<rot 88.2826 0 0>}]
 dyn_hair_5_c [{shapeType=Capsule, length=3.9255, radius=2.6628, shapeName=dyn_C_hair_1_capsule, center=<5.1655 0 -0.0758>, rotation=<rot 86.9145 -180 -180>}]
 dyn_hair_5_l [{shapeType=Capsule, length=3.9255, radius=2.6628, shapeName=dyn_C_hair_1_capsule, center=<5.1655 0 -0.0758>, rotation=<rot 86.9145 -180 -180>}]
 dyn_hair_5_r [{shapeType=Capsule, length=3.9255, radius=2.6628, shapeName=dyn_C_hair_1_capsule, center=<5.1655 0 -0.0758>, rotation=<rot 86.9145 -180 -180>}]
 dyn_hair_6_c [{shapeType=Capsule, length=4.7118, radius=2.1191, shapeName=dyn_C_hair_1_capsule, center=<4.1345 0 0.6495>, rotation=<rot 84.3481 180 180>}]
 dyn_hair_6_l [{shapeType=Capsule, length=4.7118, radius=2.1191, shapeName=dyn_C_hair_1_capsule, center=<4.1345 0 0.6495>, rotation=<rot 84.3481 180 180>}]
 dyn_hair_6_r [{shapeType=Capsule, length=4.7118, radius=2.1191, shapeName=dyn_C_hair_1_capsule, center=<4.1345 0 0.6495>, rotation=<rot 84.3481 180 180>}]
 dyn_hair_7_c [{shapeType=Capsule, length=5.5101, radius=3.5679, shapeName=dyn_hair_7_c_capsule, center=<5.3354 0 -0.062>, rotation=<rot -89.3337 180 -180>}]
 dyn_hair_7_l [{shapeType=Capsule, length=5.5101, radius=3.5679, shapeName=dyn_hair_7_l_capsule, center=<5.3354 0 -0.062>, rotation=<rot -89.3337 180 -180>}]
 dyn_hair_7_r [{shapeType=Capsule, length=5.5101, radius=3.5679, shapeName=dyn_hair_7_r_capsule, center=<5.3354 0 -0.062>, rotation=<rot -89.3337 180 -180>}]
 dyn_hair_8_c [{shapeType=Capsule, length=9.6089, radius=3.538, shapeName=dyn_hair_8_c_capsule, center=<7.659 0 0>, rotation=<rot 90 0 0>}]
 dyn_hair_8_l [{shapeType=Capsule, length=6.6675, radius=2.5159, shapeName=dyn_hair_8_l_capsule, center=<5.1746 0 0>, rotation=<rot 90 0 0>}]
 dyn_hair_8_r [{shapeType=Capsule, length=7.9159, radius=2.0413, shapeName=dyn_hair_8_r_capsule, center=<5.1746 0 0>, rotation=<rot 90 0 0>}]
 pelvis [{shapeType=Capsule, length=23.91, radius=27.2467, shapeName=pelvis_capsule, center=<0 13.7481 0>, rotation=<rot 0 -90 0>}]
 spine_02 [{shapeType=Capsule, length=4.887, radius=9.8602, shapeName=spine_02_capsule, center=<-0.2722 3.3234 -0.3025>, rotation=<rot 0 -90 0>}]
 spine_03 [{shapeType=Capsule, length=8.629, radius=9.6494, shapeName=spine_03_capsule, center=<-1.3718 2.4078 0>, rotation=<rot 0 -90 0>}]
 spine_04 [{shapeType=Capsule, length=3.4282, radius=12.5388, shapeName=spine_04_capsule, center=<0.6179 4.3829 0>, rotation=<rot 0 -90 0>}]
 spine_05 [{shapeType=Capsule, length=5.7751, radius=9.0648, shapeName=spine_04_capsule, center=<0.0586 0.9277 0>, rotation=<rot 0 -90 0>}]
}
constraints [
 {
  bone1Name=spine_03
  bone2Name=spine_02
  swing1LimitDegrees=45
  swing1Motion=Free
  swing2LimitDegrees=45
  swing2Motion=Free
  twistLimitDegrees=45
  twistMotion=Free
 }
 {
  bone1Name=spine_04
  bone2Name=spine_03
  swing1LimitDegrees=45
  swing1Motion=Free
  swing2LimitDegrees=45
  swing2Motion=Free
  twistLimitDegrees=45
  twistMotion=Free
 }
 {
  bone1Name=spine_05
  bone2Name=spine_04
  swing1LimitDegrees=45
  swing1Motion=Free
  swing2LimitDegrees=45
  swing2Motion=Free
  twistLimitDegrees=45
  twistMotion=Free
 }
 {bone1Name=dyn_hair_3_l, bone2Name=dyn_hair_2_l, swing1LimitDegrees=0, swing2LimitDegrees=30, twistLimitDegrees=0}
 {bone1Name=dyn_hair_4_l, bone2Name=dyn_hair_3_l, swing1LimitDegrees=0, swing2LimitDegrees=30, twistLimitDegrees=0}
 {bone1Name=dyn_hair_5_l, bone2Name=dyn_hair_4_l, swing1LimitDegrees=0, swing2LimitDegrees=30, twistLimitDegrees=0}
 {bone1Name=dyn_hair_6_l, bone2Name=dyn_hair_5_l, swing1LimitDegrees=0, swing2LimitDegrees=30, twistLimitDegrees=0}
 {bone1Name=dyn_hair_3_r, bone2Name=dyn_hair_2_r, swing1LimitDegrees=0, swing2LimitDegrees=30, twistLimitDegrees=0}
 {bone1Name=dyn_hair_4_r, bone2Name=dyn_hair_3_r, swing1LimitDegrees=0, swing2LimitDegrees=30, twistLimitDegrees=0}
 {bone1Name=dyn_hair_5_r, bone2Name=dyn_hair_4_r, swing1LimitDegrees=0, swing2LimitDegrees=30, twistLimitDegrees=0}
 {bone1Name=dyn_hair_6_r, bone2Name=dyn_hair_5_r, swing1LimitDegrees=0, swing2LimitDegrees=30, twistLimitDegrees=0}
 {bone1Name=dyn_hair_3_c, bone2Name=dyn_hair_2_c, swing1LimitDegrees=0, swing2LimitDegrees=30, twistLimitDegrees=0}
 {bone1Name=dyn_hair_4_c, bone2Name=dyn_hair_3_c, swing1LimitDegrees=0, swing2LimitDegrees=30, twistLimitDegrees=0}
 {bone1Name=dyn_hair_5_c, bone2Name=dyn_hair_4_c, swing1LimitDegrees=0, swing2LimitDegrees=30, twistLimitDegrees=0}
 {bone1Name=dyn_hair_6_c, bone2Name=dyn_hair_5_c, swing1LimitDegrees=0, swing2LimitDegrees=30, twistLimitDegrees=0}
 {
  bone1Name=dyn_hair_2_c
  bone2Name=spine_05
  swing1LimitDegrees=45
  swing1Motion=Free
  swing2LimitDegrees=45
  swing2Motion=Free
  twistLimitDegrees=45
  twistMotion=Free
 }
 {
  bone1Name=dyn_hair_2_l
  bone2Name=spine_05
  swing1LimitDegrees=45
  swing1Motion=Free
  swing2LimitDegrees=45
  swing2Motion=Free
  twistLimitDegrees=45
  twistMotion=Free
 }
 {
  bone1Name=dyn_hair_2_r
  bone2Name=spine_05
  swing1LimitDegrees=45
  swing1Motion=Free
  swing2LimitDegrees=45
  swing2Motion=Free
  twistLimitDegrees=45
  twistMotion=Free
 }
 {
  bone1Name=clavicle_l
  bone2Name=spine_05
  swing1LimitDegrees=45
  swing1Motion=Free
  swing2LimitDegrees=45
  swing2Motion=Free
  twistLimitDegrees=45
  twistMotion=Free
 }
 {
  bone1Name=clavicle_r
  bone2Name=spine_05
  swing1LimitDegrees=45
  swing1Motion=Free
  swing2LimitDegrees=45
  swing2Motion=Free
  twistLimitDegrees=45
  twistMotion=Free
 }
 {bone1Name=dyn_hair_7_c, bone2Name=dyn_hair_6_c, swing1LimitDegrees=0, swing2LimitDegrees=20, twistLimitDegrees=0}
 {bone1Name=dyn_hair_8_c, bone2Name=dyn_hair_7_c, swing1LimitDegrees=0, swing2LimitDegrees=20, twistLimitDegrees=0}
 {bone1Name=dyn_hair_7_l, bone2Name=dyn_hair_6_l, swing1LimitDegrees=0, swing2LimitDegrees=20, twistLimitDegrees=0}
 {bone1Name=dyn_hair_8_l, bone2Name=dyn_hair_7_l, swing1LimitDegrees=0, swing2LimitDegrees=20, twistLimitDegrees=0}
 {bone1Name=dyn_hair_7_r, bone2Name=dyn_hair_6_r, swing1LimitDegrees=0, swing2LimitDegrees=20, twistLimitDegrees=0}
 {bone1Name=dyn_hair_8_r, bone2Name=dyn_hair_7_r, swing1LimitDegrees=0, swing2LimitDegrees=20, twistLimitDegrees=0}
]
