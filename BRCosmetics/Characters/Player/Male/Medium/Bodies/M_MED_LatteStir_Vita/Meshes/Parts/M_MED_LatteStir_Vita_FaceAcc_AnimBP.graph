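# Anim Blueprint for the LatteStir Vita face accessory, adding secondary motion to the character's hair.
# Copied parent pose runs through a Control Rig, base-hair offsets and neck-hair AnimDynamics into per-bone hair AnimDynamics with averaging constraints.
# Blends switch to a cheaper center-chain simulation during melee, or to static tip offsets bypassing main hair simulation.
# EventGraph derives gravity, simulation alpha, fall-driven overlap and TurnOffDyn from movement, backbling and pickaxe state.
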AnimBlueprint /BRCosmetics/Characters/Player/Male/Medium/Bodies/M_MED_LatteStir_Vita/Meshes/Parts/M_MED_LatteStir_Vita_FaceAcc_AnimBP
target_skeleton=&/M_MED_LatteStir_Vita_FaceAcc_Skeleton
graphs [
 AnimGraph {nodes [
   RigidBody!(ComponentPose, Alpha, OverrideWorldGravity:Vector="0,0,0") -> (Pose) {
    bOverrideWorldGravity=true
    overridePhysicsAsset=&/M_MED_LatteStir_Vita_FaceAcc_Physics
    simulationSpace=BaseBoneSpace
    alphaBoolBlend {blendOption=Linear}
    alphaScaleBiasClamp {bInterpResult=true, interpSpeedDecreasing=3, interpSpeedIncreasing=3}
    baseBoneRef {boneName=spine_05}
    componentAppliedLinearAccClamp=<3000 3000 3000>
    simSpaceSettings {
     maxAngularAcceleration=100
     maxAngularVelocity=3
     maxLinearAcceleration=800
     maxLinearVelocity=400
     worldAlpha=1
     externalLinearDragV=<1 1 1>
    }
   }
   Output Pose(Result=Use cached pose 'postSimFixes'()) {layerGroup=None}
   n1 = Control Rig(Source=Copy Pose From Mesh() {bUseAttachedParent=true}) {controlRigAssetReference {blueprintRigClass=&/M_MED_LatteStir_Vita_FaceAcc_CtrlRig.M_MED_LatteStir_Vita_FaceAcc_CtrlRig_C}}
   Save cached pose 'controlRigCP'(Pose=@n1) {cacheName=controlRigCP}
   n2 = Anim Dynamics - Bone: dyn_hair_2_l!(ComponentPose=Local To Component(LocalPose=Use cached pose 'animDynamicsACP'()), Alpha=1, GravityOverride:Vector="0,0,-980") -> (Pose) {
    angularDampingOverride=0.9999
    angularSpringConstant=3000
    bAngularSpring=true
    bOverrideAngularDamping=true
    bUseGravityOverride=true
    gravityScale=0.35
    simulationSpace=RootRelative
    boundBone {boneName=dyn_hair_2_l}
    componentAppliedLinearAccClamp=<3000 3000 3000>
    componentLinearAccScale=<0.3 0.3 0.3>
    componentLinearVelScale=<0.3 0.3 0.3>
    gravityOverride {z=-980}
    physicsBodyDefinitions [{
      boundBone {boneName=dyn_hair_2_l}
      boxExtents=<25 25 25>
      constraintSetup {angularLimitsMax=<0 0 5>, angularLimitsMin=<0 -10 -5>, angularTarget=<1 0 0>}
      localJointOffset=<4 0 0>
     }]
    simSpaceSettings {maxAngularAcceleration=3, maxAngularVelocity=2, simSpaceAngularAlpha=1}
   }
   n3 = Anim Dynamics - Bone: dyn_hair_2_c!(ComponentPose=@n2, Alpha=1, GravityOverride:Vector="0,0,-980") -> (Pose) {
    angularDampingOverride=0.9999
    angularSpringConstant=3000
    bAngularSpring=true
    bOverrideAngularDamping=true
    bUseGravityOverride=true
    gravityScale=0.35
    simulationSpace=RootRelative
    boundBone {boneName=dyn_hair_2_c}
    componentAppliedLinearAccClamp=<3000 3000 3000>
    componentLinearAccScale=<0.3 0.3 0.3>
    componentLinearVelScale=<0.3 0.3 0.3>
    gravityOverride {z=-980}
    physicsBodyDefinitions [{
      boundBone {boneName=dyn_hair_2_c}
      boxExtents=<25 25 25>
      constraintSetup {angularLimitsMax=<0 0 5>, angularLimitsMin=<0 -10 -5>, angularTarget=<1 0 0>}
      localJointOffset=<4 0 0>
     }]
    simSpaceSettings {maxAngularAcceleration=3, maxAngularVelocity=2, simSpaceAngularAlpha=1}
   }
   n4 = Anim Dynamics - Bone: dyn_hair_2_r!(ComponentPose=@n3, Alpha=1, GravityOverride:Vector="0,0,-980") -> (Pose) {
    angularDampingOverride=0.9999
    angularSpringConstant=3000
    bAngularSpring=true
    bOverrideAngularDamping=true
    bUseGravityOverride=true
    gravityScale=0.35
    simulationSpace=RootRelative
    boundBone {boneName=dyn_hair_2_r}
    componentAppliedLinearAccClamp=<3000 3000 3000>
    componentLinearAccScale=<0.3 0.3 0.3>
    componentLinearVelScale=<0.3 0.3 0.3>
    gravityOverride {z=-980}
    physicsBodyDefinitions [{
      boundBone {boneName=dyn_hair_2_r}
      boxExtents=<25 25 25>
      constraintSetup {angularLimitsMax=<0 0 5>, angularLimitsMin=<0 -10 -5>, angularTarget=<1 0 0>}
      localJointOffset=<4 0 0>
     }]
    simSpaceSettings {maxAngularAcceleration=3, maxAngularVelocity=2, simSpaceAngularAlpha=1}
   }
   n5 = Anim Dynamics - Bone: dyn_hair_3_l!(ComponentPose=@n4, Alpha=1, GravityOverride:Vector="0,0,-980") -> (Pose) {
    angularDampingOverride=0.9999
    angularSpringConstant=3000
    bAngularSpring=true
    bOverrideAngularDamping=true
    bUseGravityOverride=true
    gravityScale=0.35
    simulationSpace=RootRelative
    boundBone {boneName=dyn_hair_3_l}
    componentAppliedLinearAccClamp=<3000 3000 3000>
    componentLinearAccScale=<0.3 0.3 0.3>
    componentLinearVelScale=<0.3 0.3 0.3>
    gravityOverride {z=-980}
    physicsBodyDefinitions [{
      boundBone {boneName=dyn_hair_3_l}
      boxExtents=<25 25 25>
      constraintSetup {angularLimitsMax=<0 0 10>, angularLimitsMin=<0 -30 -8>, angularTarget=<1 0 0>}
      localJointOffset=<4 0 0>
     }]
    simSpaceSettings {maxAngularAcceleration=3, maxAngularVelocity=2, simSpaceAngularAlpha=1}
   }
   n6 = Anim Dynamics - Bone: dyn_hair_3_c!(ComponentPose=@n5, Alpha=1, GravityOverride:Vector="0,0,-980") -> (Pose) {
    angularDampingOverride=0.9999
    angularSpringConstant=3000
    bAngularSpring=true
    bOverrideAngularDamping=true
    bUseGravityOverride=true
    gravityScale=0.35
    simulationSpace=RootRelative
    boundBone {boneName=dyn_hair_3_c}
    componentAppliedLinearAccClamp=<3000 3000 3000>
    componentLinearAccScale=<0.3 0.3 0.3>
    componentLinearVelScale=<0.3 0.3 0.3>
    gravityOverride {z=-980}
    physicsBodyDefinitions [{
      boundBone {boneName=dyn_hair_3_c}
      boxExtents=<25 25 25>
      constraintSetup {angularLimitsMax=<0 0 10>, angularLimitsMin=<0 -30 -8>, angularTarget=<1 0 0>}
      localJointOffset=<4 0 0>
     }]
    simSpaceSettings {maxAngularAcceleration=3, maxAngularVelocity=2, simSpaceAngularAlpha=1}
   }
   n7 = Anim Dynamics - Bone: dyn_hair_3_r!(ComponentPose=@n6, Alpha=1, GravityOverride:Vector="0,0,-980") -> (Pose) {
    angularDampingOverride=0.9999
    angularSpringConstant=3000
    bAngularSpring=true
    bOverrideAngularDamping=true
    bUseGravityOverride=true
    gravityScale=0.35
    simulationSpace=RootRelative
    boundBone {boneName=dyn_hair_3_r}
    componentAppliedLinearAccClamp=<3000 3000 3000>
    componentLinearAccScale=<0.3 0.3 0.3>
    componentLinearVelScale=<0.3 0.3 0.3>
    gravityOverride {z=-980}
    physicsBodyDefinitions [{
      boundBone {boneName=dyn_hair_3_r}
      boxExtents=<25 25 25>
      constraintSetup {angularLimitsMax=<0 0 10>, angularLimitsMin=<0 -30 -8>, angularTarget=<1 0 0>}
      localJointOffset=<4 0 0>
     }]
    simSpaceSettings {maxAngularAcceleration=3, maxAngularVelocity=2, simSpaceAngularAlpha=1}
   }
   n8 = Anim Dynamics - Bone: dyn_hair_4_l!(ComponentPose=@n7, Alpha=1, GravityOverride:Vector="0,0,-980") -> (Pose) {
    angularDampingOverride=0.9999
    angularSpringConstant=3000
    bAngularSpring=true
    bOverrideAngularDamping=true
    bUseGravityOverride=true
    gravityScale=0.35
    simulationSpace=RootRelative
    boundBone {boneName=dyn_hair_4_l}
    componentAppliedLinearAccClamp=<3000 3000 3000>
    componentLinearAccScale=<0.3 0.3 0.3>
    componentLinearVelScale=<0.3 0.3 0.3>
    gravityOverride {z=-980}
    physicsBodyDefinitions [{
      boundBone {boneName=dyn_hair_4_l}
      boxExtents=<25 25 25>
      constraintSetup {angularLimitsMax=<0 0 10>, angularLimitsMin=<0 -25 -8>, angularTarget=<1 0 0>}
      localJointOffset=<4 0 0>
     }]
    simSpaceSettings {maxAngularAcceleration=3, maxAngularVelocity=2, simSpaceAngularAlpha=1}
   }
   n9 = Anim Dynamics - Bone: dyn_hair_4_c!(ComponentPose=@n8, Alpha=1, GravityOverride:Vector="0,0,-980") -> (Pose) {
    angularDampingOverride=0.9999
    angularSpringConstant=3000
    bAngularSpring=true
    bOverrideAngularDamping=true
    bUseGravityOverride=true
    gravityScale=0.35
    simulationSpace=RootRelative
    boundBone {boneName=dyn_hair_4_c}
    componentAppliedLinearAccClamp=<3000 3000 3000>
    componentLinearAccScale=<0.3 0.3 0.3>
    componentLinearVelScale=<0.3 0.3 0.3>
    gravityOverride {z=-980}
    physicsBodyDefinitions [{
      boundBone {boneName=dyn_hair_4_c}
      boxExtents=<25 25 25>
      constraintSetup {angularLimitsMax=<0 0 10>, angularLimitsMin=<0 -25 -8>, angularTarget=<1 0 0>}
      localJointOffset=<4 0 0>
     }]
    simSpaceSettings {maxAngularAcceleration=3, maxAngularVelocity=2, simSpaceAngularAlpha=1}
   }
   n10 = Anim Dynamics - Bone: dyn_hair_4_r!(ComponentPose=@n9, Alpha=1, GravityOverride:Vector="0,0,-980") -> (Pose) {
    angularDampingOverride=0.9999
    angularSpringConstant=3000
    bAngularSpring=true
    bOverrideAngularDamping=true
    bUseGravityOverride=true
    gravityScale=0.35
    simulationSpace=RootRelative
    boundBone {boneName=dyn_hair_4_r}
    componentAppliedLinearAccClamp=<3000 3000 3000>
    componentLinearAccScale=<0.3 0.3 0.3>
    componentLinearVelScale=<0.3 0.3 0.3>
    gravityOverride {z=-980}
    physicsBodyDefinitions [{
      boundBone {boneName=dyn_hair_4_r}
      boxExtents=<25 25 25>
      constraintSetup {angularLimitsMax=<0 0 10>, angularLimitsMin=<0 -25 -8>, angularTarget=<1 0 0>}
      localJointOffset=<4 0 0>
     }]
    simSpaceSettings {maxAngularAcceleration=3, maxAngularVelocity=2, simSpaceAngularAlpha=1}
   }
   n11 = Anim Dynamics - Bone: dyn_hair_5_l!(ComponentPose=@n10, Alpha=1, GravityOverride:Vector="0,0,-980") -> (Pose) {
    angularDampingOverride=0.9999
    angularSpringConstant=3000
    bAngularSpring=true
    bOverrideAngularDamping=true
    bUseGravityOverride=true
    gravityScale=0.35
    simulationSpace=RootRelative
    boundBone {boneName=dyn_hair_5_l}
    componentAppliedLinearAccClamp=<3000 3000 3000>
    componentLinearAccScale=<0.3 0.3 0.3>
    componentLinearVelScale=<0.3 0.3 0.3>
    gravityOverride {z=-980}
    physicsBodyDefinitions [{
      boundBone {boneName=dyn_hair_5_l}
      boxExtents=<25 25 25>
      constraintSetup {angularLimitsMax=<0 0 10>, angularLimitsMin=<0 -20 -8>, angularTarget=<1 0 0>}
      localJointOffset=<4 0 0>
     }]
    simSpaceSettings {maxAngularAcceleration=3, maxAngularVelocity=2, simSpaceAngularAlpha=1}
   }
   n12 = Anim Dynamics - Bone: dyn_hair_5_c!(ComponentPose=@n11, Alpha=1, GravityOverride:Vector="0,0,-980") -> (Pose) {
    angularDampingOverride=0.9999
    angularSpringConstant=3000
    bAngularSpring=true
    bOverrideAngularDamping=true
    bUseGravityOverride=true
    gravityScale=0.35
    simulationSpace=RootRelative
    boundBone {boneName=dyn_hair_5_c}
    componentAppliedLinearAccClamp=<3000 3000 3000>
    componentLinearAccScale=<0.3 0.3 0.3>
    componentLinearVelScale=<0.3 0.3 0.3>
    gravityOverride {z=-980}
    physicsBodyDefinitions [{
      boundBone {boneName=dyn_hair_5_c}
      boxExtents=<25 25 25>
      constraintSetup {angularLimitsMax=<0 0 10>, angularLimitsMin=<0 -20 -8>, angularTarget=<1 0 0>}
      localJointOffset=<4 0 0>
     }]
    simSpaceSettings {maxAngularAcceleration=3, maxAngularVelocity=2, simSpaceAngularAlpha=1}
   }
   n13 = Anim Dynamics - Bone: dyn_hair_5_r!(ComponentPose=@n12, Alpha=1, GravityOverride:Vector="0,0,-980") -> (Pose) {
    angularDampingOverride=0.9999
    angularSpringConstant=3000
    bAngularSpring=true
    bOverrideAngularDamping=true
    bUseGravityOverride=true
    gravityScale=0.35
    simulationSpace=RootRelative
    boundBone {boneName=dyn_hair_5_r}
    componentAppliedLinearAccClamp=<3000 3000 3000>
    componentLinearAccScale=<0.3 0.3 0.3>
    componentLinearVelScale=<0.3 0.3 0.3>
    gravityOverride {z=-980}
    physicsBodyDefinitions [{
      boundBone {boneName=dyn_hair_5_r}
      boxExtents=<25 25 25>
      constraintSetup {angularLimitsMax=<0 0 10>, angularLimitsMin=<0 -20 -8>, angularTarget=<1 0 0>}
      localJointOffset=<4 0 0>
     }]
    simSpaceSettings {maxAngularAcceleration=3, maxAngularVelocity=2, simSpaceAngularAlpha=1}
   }
   n14 = Anim Dynamics - Bone: dyn_hair_6_l!(ComponentPose=@n13, Alpha=1, GravityOverride:Vector="0,0,-980") -> (Pose) {
    angularDampingOverride=0.9999
    angularSpringConstant=3000
    bAngularSpring=true
    bOverrideAngularDamping=true
    bUseGravityOverride=true
    gravityScale=0.35
    simulationSpace=RootRelative
    boundBone {boneName=dyn_hair_6_l}
    componentAppliedLinearAccClamp=<3000 3000 3000>
    componentLinearAccScale=<0.3 0.3 0.3>
    componentLinearVelScale=<0.3 0.3 0.3>
    gravityOverride {z=-980}
    physicsBodyDefinitions [{
      boundBone {boneName=dyn_hair_6_l}
      boxExtents=<25 25 25>
      constraintSetup {angularLimitsMax=<0 0 10>, angularLimitsMin=<0 -15 -8>, angularTarget=<1 0 0>}
      localJointOffset=<4 0 0>
     }]
    simSpaceSettings {maxAngularAcceleration=3, maxAngularVelocity=2, simSpaceAngularAlpha=1}
   }
   n15 = Anim Dynamics - Bone: dyn_hair_6_c!(ComponentPose=@n14, Alpha=1, GravityOverride:Vector="0,0,-980") -> (Pose) {
    angularDampingOverride=0.9999
    angularSpringConstant=3000
    bAngularSpring=true
    bOverrideAngularDamping=true
    bUseGravityOverride=true
    gravityScale=0.35
    simulationSpace=RootRelative
    boundBone {boneName=dyn_hair_6_c}
    componentAppliedLinearAccClamp=<3000 3000 3000>
    componentLinearAccScale=<0.3 0.3 0.3>
    componentLinearVelScale=<0.3 0.3 0.3>
    gravityOverride {z=-980}
    physicsBodyDefinitions [{
      boundBone {boneName=dyn_hair_6_c}
      boxExtents=<25 25 25>
      constraintSetup {angularLimitsMax=<0 0 10>, angularLimitsMin=<0 -15 -8>, angularTarget=<1 0 0>}
      localJointOffset=<4 0 0>
     }]
    simSpaceSettings {maxAngularAcceleration=3, maxAngularVelocity=2, simSpaceAngularAlpha=1}
   }
   n16 = Anim Dynamics - Bone: dyn_hair_6_r!(ComponentPose=@n15, Alpha=1, GravityOverride:Vector="0,0,-980") -> (Pose) {
    angularDampingOverride=0.9999
    angularSpringConstant=3000
    bAngularSpring=true
    bOverrideAngularDamping=true
    bUseGravityOverride=true
    bUseSphericalLimits=true
    gravityScale=0.35
    simulationSpace=RootRelative
    boundBone {boneName=dyn_hair_6_r}
    componentAppliedLinearAccClamp=<3000 3000 3000>
    componentLinearAccScale=<0.3 0.3 0.3>
    componentLinearVelScale=<0.3 0.3 0.3>
    gravityOverride {z=-980}
    physicsBodyDefinitions [{
      boundBone {boneName=dyn_hair_6_r}
      boxExtents=<25 25 25>
      constraintSetup {angularLimitsMax=<0 0 10>, angularLimitsMin=<0 -15 -8>, angularTarget=<1 0 0>}
      localJointOffset=<4 0 0>
     }]
    simSpaceSettings {maxAngularAcceleration=3, maxAngularVelocity=2, simSpaceAngularAlpha=1}
    sphericalLimits [{limitRadius=12, drivingBone {boneName=upperarm_r}, sphereLocalOffset=<-20 0 0>}]
   }
   n17 = Anim Dynamics - Bone: dyn_hair_7_l!(ComponentPose=@n16, Alpha=1, GravityOverride:Vector="0,0,-980") -> (Pose) {
    angularDampingOverride=0.9999
    angularSpringConstant=3000
    bAngularSpring=true
    bOverrideAngularDamping=true
    bUseGravityOverride=true
    gravityScale=0.35
    simulationSpace=RootRelative
    boundBone {boneName=dyn_hair_7_l}
    componentAppliedLinearAccClamp=<3000 3000 3000>
    componentLinearAccScale=<0.3 0.3 0.3>
    componentLinearVelScale=<0.3 0.3 0.3>
    gravityOverride {z=-980}
    physicsBodyDefinitions [{
      boundBone {boneName=dyn_hair_7_l}
      boxExtents=<25 25 25>
      constraintSetup {angularLimitsMax=<0 0 10>, angularLimitsMin=<0 -10 -10>, angularTarget=<1 0 0>}
      localJointOffset=<4 0 0>
     }]
    simSpaceSettings {maxAngularAcceleration=3, maxAngularVelocity=2, simSpaceAngularAlpha=1}
   }
   n18 = Anim Dynamics - Bone: dyn_hair_7_c!(ComponentPose=@n17, Alpha=1, GravityOverride:Vector="0,0,-980") -> (Pose) {
    angularDampingOverride=0.9999
    angularSpringConstant=3000
    bAngularSpring=true
    bOverrideAngularDamping=true
    bUseGravityOverride=true
    gravityScale=0.35
    simulationSpace=RootRelative
    boundBone {boneName=dyn_hair_7_c}
    componentAppliedLinearAccClamp=<3000 3000 3000>
    componentLinearAccScale=<0.3 0.3 0.3>
    componentLinearVelScale=<0.3 0.3 0.3>
    gravityOverride {z=-980}
    physicsBodyDefinitions [{
      boundBone {boneName=dyn_hair_7_c}
      boxExtents=<25 25 25>
      constraintSetup {angularLimitsMax=<0 0 10>, angularLimitsMin=<0 -10 -10>, angularTarget=<1 0 0>}
      localJointOffset=<4 0 0>
     }]
    simSpaceSettings {maxAngularAcceleration=3, maxAngularVelocity=2, simSpaceAngularAlpha=1}
   }
   n19 = Anim Dynamics - Bone: dyn_hair_7_r!(ComponentPose=@n18, Alpha=1, GravityOverride:Vector="0,0,-980") -> (Pose) {
    angularDampingOverride=0.9999
    angularSpringConstant=3000
    bAngularSpring=true
    bOverrideAngularDamping=true
    bUseGravityOverride=true
    gravityScale=0.35
    simulationSpace=RootRelative
    boundBone {boneName=dyn_hair_7_r}
    componentAppliedLinearAccClamp=<3000 3000 3000>
    componentLinearAccScale=<0.3 0.3 0.3>
    componentLinearVelScale=<0.3 0.3 0.3>
    gravityOverride {z=-980}
    physicsBodyDefinitions [{
      boundBone {boneName=dyn_hair_7_r}
      boxExtents=<25 25 25>
      constraintSetup {angularLimitsMax=<0 0 10>, angularLimitsMin=<0 -10 -10>, angularTarget=<1 0 0>}
      localJointOffset=<4 0 0>
     }]
    simSpaceSettings {maxAngularAcceleration=3, maxAngularVelocity=2, simSpaceAngularAlpha=1}
   }
   n20 = "Constraint - dyn_hair_4_c (PP)"!(ComponentPose=@n19, Alpha=1, ConstraintWeights_0=0.5, ConstraintWeights_1:float=0.5) -> (Pose) {
    boneToModify {boneName=dyn_hair_4_c}
    constraintSetup [{targetBone {boneName=dyn_hair_4_r}}, {targetBone {boneName=dyn_hair_4_l}}]
    constraintWeights [1,1]
   }
   n21 = "Constraint - dyn_hair_5_c (PP)"!(ComponentPose=@n20, Alpha=1, ConstraintWeights_0=0.5, ConstraintWeights_1:float=0.5) -> (Pose) {
    boneToModify {boneName=dyn_hair_5_c}
    constraintSetup [{targetBone {boneName=dyn_hair_5_r}}, {targetBone {boneName=dyn_hair_5_l}}]
    constraintWeights [1,1]
   }
   n22 = "Constraint - dyn_hair_6_c (PP)"!(ComponentPose=@n21, Alpha=1, ConstraintWeights_0=0.5, ConstraintWeights_1:float=0.5) -> (Pose) {
    boneToModify {boneName=dyn_hair_6_c}
    constraintSetup [{targetBone {boneName=dyn_hair_6_r}}, {targetBone {boneName=dyn_hair_6_l}}]
    constraintWeights [1,1]
   }
   n23 = "Constraint - dyn_hair_7_c (PP)"!(ComponentPose=@n22, Alpha=1, ConstraintWeights_0=0.5, ConstraintWeights_1:float=0.5) -> (Pose) {
    boneToModify {boneName=dyn_hair_7_c}
    constraintSetup [{targetBone {boneName=dyn_hair_7_r}}, {targetBone {boneName=dyn_hair_7_l}}]
    constraintWeights [1,1]
   }
   n24 = Component To Local(ComponentPose=@n23)
   n25 = Use cached pose 'CRCache_PlayingMeele'()
   n26 = Get bIsPlayingMeleeAnim()
   n27 = "Transform (Modify) Bone - Bone: dyn_hair_7_l"!(ComponentPose=Local To Component(LocalPose=Use cached pose 'animDynamicsACP'()), bAlphaBoolEnabled:bool=True, Rotation="20,0,0") -> (Pose) {
    alphaInputType=Bool
    rotationMode=BMM_Additive
    rotationSpace=BCS_BoneSpace
    translationMode=BMM_Additive
    translationSpace=BCS_BoneSpace
    boneToModify {boneName=dyn_hair_7_l}
    rotation {pitch=20}
    translation {x=2}
   }
   n28 = "Transform (Modify) Bone - Bone: dyn_hair_7_c"!(ComponentPose=@n27, bAlphaBoolEnabled:bool=True, Rotation="20,0,0") -> (Pose) {
    alphaInputType=Bool
    rotationMode=BMM_Additive
    rotationSpace=BCS_BoneSpace
    translationMode=BMM_Additive
    translationSpace=BCS_BoneSpace
    boneToModify {boneName=dyn_hair_7_c}
    rotation {pitch=20}
    translation {x=2}
   }
   n29 = "Transform (Modify) Bone - Bone: dyn_hair_7_r"!(ComponentPose=@n28, bAlphaBoolEnabled:bool=True, Rotation="20,0,0") -> (Pose) {
    alphaInputType=Bool
    rotationMode=BMM_Additive
    rotationSpace=BCS_BoneSpace
    translationMode=BMM_Additive
    translationSpace=BCS_BoneSpace
    boneToModify {boneName=dyn_hair_7_r}
    rotation {pitch=20}
    translation {x=2}
   }
   n30 = "Transform (Modify) Bone - Bone: dyn_hair_8_l"!(ComponentPose=@n29, bAlphaBoolEnabled:bool=True, Rotation="20,0,0") -> (Pose) {
    alphaInputType=Bool
    rotationMode=BMM_Additive
    rotationSpace=BCS_BoneSpace
    translationMode=BMM_Additive
    translationSpace=BCS_BoneSpace
    boneToModify {boneName=dyn_hair_8_l}
    rotation {pitch=20}
   }
   n31 = "Transform (Modify) Bone - Bone: dyn_hair_8_c"!(ComponentPose=@n30, bAlphaBoolEnabled:bool=True, Rotation="20,0,0") -> (Pose) {
    alphaInputType=Bool
    rotationMode=BMM_Additive
    rotationSpace=BCS_BoneSpace
    translationMode=BMM_Additive
    translationSpace=BCS_BoneSpace
    boneToModify {boneName=dyn_hair_8_c}
    rotation {pitch=20}
   }
   n32 = "Transform (Modify) Bone - Bone: dyn_hair_8_r"!(ComponentPose=@n31, bAlphaBoolEnabled:bool=True, Rotation="20,0,0") -> (Pose) {
    alphaInputType=Bool
    rotationMode=BMM_Additive
    rotationSpace=BCS_BoneSpace
    translationMode=BMM_Additive
    translationSpace=BCS_BoneSpace
    boneToModify {boneName=dyn_hair_8_r}
    rotation {pitch=20}
   }
   Save cached pose 'dynamicsCP'(Pose=Blend Poses by bool(
     BlendPose_0=Component To Local(ComponentPose=@n32)
     BlendPose_1=Blend Poses by bool(BlendPose_0=@n25, BlendPose_1=@n24, BlendTime_0=0.3, BlendTime_1=0.8, bActiveValue=@n26)
    ) {blendType=Linear}) {cacheName=dynamicsCP}
   n33 = "Transform (Modify) Bone - Bone: base_hair_1_l"!(ComponentPose=Local To Component(LocalPose=Use cached pose 'controlRigCP'()), Alpha=1, Translation="0,0,0", Rotation="10,0,0") -> (Pose) {
    rotationMode=BMM_Additive
    rotationSpace=BCS_BoneSpace
    translationMode=BMM_Additive
    translationSpace=BCS_BoneSpace
    alphaBoolBlend {blendOption=Linear}
    boneToModify {boneName=base_hair_1_l}
    rotation {pitch=10}
   }
   n34 = "Transform (Modify) Bone - Bone: base_hair_1_c"!(ComponentPose=@n33, Alpha=1, Translation="0,0,0", Rotation="10,0,0") -> (Pose) {
    rotationMode=BMM_Additive
    rotationSpace=BCS_BoneSpace
    translationMode=BMM_Additive
    translationSpace=BCS_BoneSpace
    alphaBoolBlend {blendOption=Linear}
    boneToModify {boneName=base_hair_1_c}
    rotation {pitch=10}
   }
   n35 = "Transform (Modify) Bone - Bone: base_hair_1_r"!(ComponentPose=@n34, Alpha=1, Translation="0,0,0", Rotation="10,0,0") -> (Pose) {
    rotationMode=BMM_Additive
    rotationSpace=BCS_BoneSpace
    translationMode=BMM_Additive
    translationSpace=BCS_BoneSpace
    alphaBoolBlend {blendOption=Linear}
    boneToModify {boneName=base_hair_1_r}
    rotation {pitch=10}
   }
   Save cached pose 'preSimTweaks'(Pose=Component To Local(ComponentPose=@n35)) {cacheName=preSimTweaks}
   n36 = Anim Dynamics - Bone: dyn_hairNeck_01_l!(ComponentPose=Local To Component(LocalPose=Use cached pose 'preSimTweaks'()), Alpha=1, GravityOverride:Vector="0,0,-980") -> (Pose) {
    angularDampingOverride=0.9999
    angularSpringConstant=6000
    bAngularSpring=true
    bOverrideAngularDamping=true
    bUseGravityOverride=true
    gravityScale=0.35
    simulationSpace=RootRelative
    boundBone {boneName=dyn_hairNeck_01_l}
    componentAppliedLinearAccClamp=<3000 3000 3000>
    componentLinearAccScale=<0.2 0.2 0.2>
    componentLinearVelScale=<0.2 0.2 0.2>
    gravityOverride {z=-980}
    physicsBodyDefinitions [{
      boundBone {boneName=dyn_hairNeck_01_l}
      boxExtents=<15 20 20>
      constraintSetup {angularLimitsMax=<0 25 25>, angularLimitsMin=<0 -25 -25>, angularTarget=<1 0 0>}
      localJointOffset=<4 0 0>
     }]
   }
   n37 = Anim Dynamics - Bone: dyn_hairNeck_02_l!(ComponentPose=@n36, Alpha=1, GravityOverride:Vector="0,0,-980") -> (Pose) {
    angularDampingOverride=0.9999
    angularSpringConstant=6000
    bAngularSpring=true
    bOverrideAngularDamping=true
    bUseGravityOverride=true
    gravityScale=0.35
    simulationSpace=RootRelative
    boundBone {boneName=dyn_hairNeck_02_l}
    componentAppliedLinearAccClamp=<3000 3000 3000>
    componentLinearAccScale=<0.2 0.2 0.2>
    componentLinearVelScale=<0.2 0.2 0.2>
    gravityOverride {z=-980}
    physicsBodyDefinitions [{
      boundBone {boneName=dyn_hairNeck_02_l}
      boxExtents=<15 20 20>
      constraintSetup {angularLimitsMax=<0 25 25>, angularLimitsMin=<0 -25 -25>, angularTarget=<1 0 0>}
      localJointOffset=<4 0 0>
     }]
   }
   n38 = Anim Dynamics - Bone: dyn_hairNeck_03_l!(ComponentPose=@n37, Alpha=1, GravityOverride:Vector="0,0,-980") -> (Pose) {
    angularDampingOverride=0.9999
    angularSpringConstant=6000
    bAngularSpring=true
    bOverrideAngularDamping=true
    bUseGravityOverride=true
    gravityScale=0.35
    simulationSpace=RootRelative
    boundBone {boneName=dyn_hairNeck_03_l}
    componentAppliedLinearAccClamp=<3000 3000 3000>
    componentLinearAccScale=<0.2 0.2 0.2>
    componentLinearVelScale=<0.2 0.2 0.2>
    gravityOverride {z=-980}
    physicsBodyDefinitions [{
      boundBone {boneName=dyn_hairNeck_03_l}
      boxExtents=<15 20 20>
      constraintSetup {angularLimitsMax=<0 25 25>, angularLimitsMin=<0 -25 -25>, angularTarget=<1 0 0>}
      localJointOffset=<4 0 0>
     }]
   }
   n39 = Anim Dynamics - Bone: dyn_hairNeck_01_r!(ComponentPose=@n38, Alpha=1, GravityOverride:Vector="0,0,-980") -> (Pose) {
    angularDampingOverride=0.9999
    angularSpringConstant=6000
    bAngularSpring=true
    bOverrideAngularDamping=true
    bUseGravityOverride=true
    gravityScale=0.35
    simulationSpace=RootRelative
    boundBone {boneName=dyn_hairNeck_01_r}
    componentAppliedLinearAccClamp=<3000 3000 3000>
    componentLinearAccScale=<0.2 0.2 0.2>
    componentLinearVelScale=<0.2 0.2 0.2>
    gravityOverride {z=-980}
    physicsBodyDefinitions [{
      boundBone {boneName=dyn_hairNeck_01_r}
      boxExtents=<15 20 20>
      constraintSetup {angularLimitsMax=<0 25 25>, angularLimitsMin=<0 -25 -25>, angularTarget=<1 0 0>}
      localJointOffset=<4 0 0>
     }]
   }
   n40 = Anim Dynamics - Bone: dyn_hairNeck_02_r!(ComponentPose=@n39, Alpha=1, GravityOverride:Vector="0,0,-980") -> (Pose) {
    angularDampingOverride=0.9999
    angularSpringConstant=6000
    bAngularSpring=true
    bOverrideAngularDamping=true
    bUseGravityOverride=true
    gravityScale=0.35
    simulationSpace=RootRelative
    boundBone {boneName=dyn_hairNeck_02_r}
    componentAppliedLinearAccClamp=<3000 3000 3000>
    componentLinearAccScale=<0.2 0.2 0.2>
    componentLinearVelScale=<0.2 0.2 0.2>
    gravityOverride {z=-980}
    physicsBodyDefinitions [{
      boundBone {boneName=dyn_hairNeck_02_r}
      boxExtents=<15 20 20>
      constraintSetup {angularLimitsMax=<0 25 25>, angularLimitsMin=<0 -25 -25>, angularTarget=<1 0 0>}
      localJointOffset=<4 0 0>
     }]
   }
   n41 = Anim Dynamics - Bone: dyn_hairNeck_03_r!(ComponentPose=@n40, Alpha=1, GravityOverride:Vector="0,0,-980") -> (Pose) {
    angularDampingOverride=0.9999
    angularSpringConstant=6000
    bAngularSpring=true
    bOverrideAngularDamping=true
    bUseGravityOverride=true
    gravityScale=0.35
    simulationSpace=RootRelative
    boundBone {boneName=dyn_hairNeck_03_r}
    componentAppliedLinearAccClamp=<3000 3000 3000>
    componentLinearAccScale=<0.2 0.2 0.2>
    componentLinearVelScale=<0.2 0.2 0.2>
    gravityOverride {z=-980}
    physicsBodyDefinitions [{
      boundBone {boneName=dyn_hairNeck_03_r}
      boxExtents=<15 20 20>
      constraintSetup {angularLimitsMax=<0 25 25>, angularLimitsMin=<0 -25 -25>, angularTarget=<1 0 0>}
      localJointOffset=<4 0 0>
     }]
   }
   n42 = Anim Dynamics - Bone: dyn_hairNeck_04_r!(ComponentPose=@n41, Alpha=1, GravityOverride:Vector="0,0,-980") -> (Pose) {
    angularDampingOverride=0.9999
    angularSpringConstant=6000
    bAngularSpring=true
    bOverrideAngularDamping=true
    bUseGravityOverride=true
    gravityScale=0.35
    simulationSpace=RootRelative
    boundBone {boneName=dyn_hairNeck_04_r}
    componentAppliedLinearAccClamp=<3000 3000 3000>
    componentLinearAccScale=<0.2 0.2 0.2>
    componentLinearVelScale=<0.2 0.2 0.2>
    gravityOverride {z=-980}
    physicsBodyDefinitions [{
      boundBone {boneName=dyn_hairNeck_04_r}
      boxExtents=<15 20 20>
      constraintSetup {angularLimitsMax=<0 25 25>, angularLimitsMin=<0 -25 -25>, angularTarget=<1 0 0>}
      localJointOffset=<4 0 0>
     }]
   }
   Save cached pose 'animDynamicsACP'(Pose=Component To Local(ComponentPose=@n42)) {cacheName=animDynamicsACP}
   n43 = Anim Dynamics - Bone: dyn_hair_2_c(ComponentPose=Local To Component(LocalPose=Use cached pose 'preSimTweaks'()), Alpha=1, GravityOverride="0,0,-980") {
    angularDampingOverride=0.9999
    angularSpringConstant=3000
    bAngularSpring=true
    bOverrideAngularDamping=true
    bUseGravityOverride=true
    gravityScale=0.35
    simulationSpace=RootRelative
    boundBone {boneName=dyn_hair_2_c}
    componentAppliedLinearAccClamp=<3000 3000 3000>
    componentLinearAccScale=<0.3 0.3 0.3>
    componentLinearVelScale=<0.3 0.3 0.3>
    gravityOverride {z=-980}
    physicsBodyDefinitions [{
      boundBone {boneName=dyn_hair_2_c}
      boxExtents=<25 25 25>
      constraintSetup {angularLimitsMax=<0 0 5>, angularLimitsMin=<0 -10 -5>, angularTarget=<1 0 0>}
      localJointOffset=<4 0 0>
     }]
   }
   n44 = Anim Dynamics - Bone: dyn_hair_3_c(ComponentPose=@n43, Alpha=1, GravityOverride="0,0,-980") {
    angularDampingOverride=0.9999
    angularSpringConstant=3000
    bAngularSpring=true
    bOverrideAngularDamping=true
    bUseGravityOverride=true
    gravityScale=0.35
    simulationSpace=RootRelative
    boundBone {boneName=dyn_hair_3_c}
    componentAppliedLinearAccClamp=<3000 3000 3000>
    componentLinearAccScale=<0.3 0.3 0.3>
    componentLinearVelScale=<0.3 0.3 0.3>
    gravityOverride {z=-980}
    physicsBodyDefinitions [{
      boundBone {boneName=dyn_hair_3_c}
      boxExtents=<25 25 25>
      constraintSetup {angularLimitsMax=<0 0 5>, angularLimitsMin=<0 -30 -5>, angularTarget=<1 0 0>}
      localJointOffset=<4 0 0>
     }]
   }
   n45 = Anim Dynamics - Bone: dyn_hair_4_c(ComponentPose=@n44, Alpha=1, GravityOverride="0,0,-980") {
    angularDampingOverride=0.9999
    angularSpringConstant=3000
    bAngularSpring=true
    bOverrideAngularDamping=true
    bUseGravityOverride=true
    gravityScale=0.35
    simulationSpace=RootRelative
    boundBone {boneName=dyn_hair_4_c}
    componentAppliedLinearAccClamp=<3000 3000 3000>
    componentLinearAccScale=<0.3 0.3 0.3>
    componentLinearVelScale=<0.3 0.3 0.3>
    gravityOverride {z=-980}
    physicsBodyDefinitions [{
      boundBone {boneName=dyn_hair_4_c}
      boxExtents=<25 25 25>
      constraintSetup {angularLimitsMax=<0 0 5>, angularLimitsMin=<0 -25 -5>, angularTarget=<1 0 0>}
      localJointOffset=<4 0 0>
     }]
   }
   n46 = Anim Dynamics - Bone: dyn_hair_5_c(ComponentPose=@n45, Alpha=1, GravityOverride="0,0,-980") {
    angularDampingOverride=0.9999
    angularSpringConstant=3000
    bAngularSpring=true
    bOverrideAngularDamping=true
    bUseGravityOverride=true
    gravityScale=0.35
    simulationSpace=RootRelative
    boundBone {boneName=dyn_hair_5_c}
    componentAppliedLinearAccClamp=<3000 3000 3000>
    componentLinearAccScale=<0.3 0.3 0.3>
    componentLinearVelScale=<0.3 0.3 0.3>
    gravityOverride {z=-980}
    physicsBodyDefinitions [{
      boundBone {boneName=dyn_hair_5_c}
      boxExtents=<25 25 25>
      constraintSetup {angularLimitsMax=<0 0 5>, angularLimitsMin=<0 -20 -5>, angularTarget=<1 0 0>}
      localJointOffset=<4 0 0>
     }]
   }
   n47 = Anim Dynamics - Bone: dyn_hair_6_c(ComponentPose=@n46, Alpha=1, GravityOverride="0,0,-980") {
    angularDampingOverride=0.9999
    angularSpringConstant=3000
    bAngularSpring=true
    bOverrideAngularDamping=true
    bUseGravityOverride=true
    gravityScale=0.35
    simulationSpace=RootRelative
    boundBone {boneName=dyn_hair_6_c}
    componentAppliedLinearAccClamp=<3000 3000 3000>
    componentLinearAccScale=<0.3 0.3 0.3>
    componentLinearVelScale=<0.3 0.3 0.3>
    gravityOverride {z=-980}
    physicsBodyDefinitions [{
      boundBone {boneName=dyn_hair_6_c}
      boxExtents=<25 25 25>
      constraintSetup {angularLimitsMax=<0 0 5>, angularLimitsMin=<0 -15 -5>, angularTarget=<1 0 0>}
      localJointOffset=<4 0 0>
     }]
   }
   n48 = Anim Dynamics - Bone: dyn_hair_7_c(ComponentPose=@n47, Alpha=1, GravityOverride="0,0,-980") {
    angularDampingOverride=0.9999
    angularSpringConstant=3000
    bAngularSpring=true
    bOverrideAngularDamping=true
    bUseGravityOverride=true
    gravityScale=0.35
    simulationSpace=RootRelative
    boundBone {boneName=dyn_hair_7_c}
    componentAppliedLinearAccClamp=<3000 3000 3000>
    componentLinearAccScale=<0.3 0.3 0.3>
    componentLinearVelScale=<0.3 0.3 0.3>
    gravityOverride {z=-980}
    physicsBodyDefinitions [{
      boundBone {boneName=dyn_hair_7_c}
      boxExtents=<25 25 25>
      constraintSetup {angularLimitsMax=<0 0 5>, angularLimitsMin=<0 -10 -5>, angularTarget=<1 0 0>}
      localJointOffset=<4 0 0>
     }]
   }
   n49 = "Constraint - dyn_hair_2_l (P)"!(ComponentPose=@n48, ConstraintWeights_0=1) -> (Pose) {alpha=0.3, boneToModify {boneName=dyn_hair_2_l}, constraintSetup [{targetBone {boneName=dyn_hair_2_c}}], constraintWeights [1]}
   n50 = "Constraint - dyn_hair_3_l (P)"!(ComponentPose=@n49, ConstraintWeights_0=1) -> (Pose) {alpha=0.3, boneToModify {boneName=dyn_hair_3_l}, constraintSetup [{targetBone {boneName=dyn_hair_3_c}}], constraintWeights [1]}
   n51 = "Constraint - dyn_hair_4_l (P)"!(ComponentPose=@n50, ConstraintWeights_0=1) -> (Pose) {alpha=0.3, boneToModify {boneName=dyn_hair_4_l}, constraintSetup [{targetBone {boneName=dyn_hair_4_c}}], constraintWeights [1]}
   n52 = "Constraint - dyn_hair_5_l (P)"!(ComponentPose=@n51, ConstraintWeights_0=1) -> (Pose) {alpha=0.3, boneToModify {boneName=dyn_hair_5_l}, constraintSetup [{targetBone {boneName=dyn_hair_5_c}}], constraintWeights [1]}
   n53 = "Constraint - dyn_hair_6_l (P)"!(ComponentPose=@n52, ConstraintWeights_0=1) -> (Pose) {boneToModify {boneName=dyn_hair_6_l}, constraintSetup [{targetBone {boneName=dyn_hair_6_c}}], constraintWeights [1]}
   n54 = "Constraint - dyn_hair_7_l (P)"!(ComponentPose=@n53, ConstraintWeights_0=1) -> (Pose) {boneToModify {boneName=dyn_hair_7_l}, constraintSetup [{targetBone {boneName=dyn_hair_7_c}}], constraintWeights [1]}
   n55 = "Constraint - dyn_hair_2_r (P)"!(ComponentPose=@n54, ConstraintWeights_0=1) -> (Pose) {alpha=0.3, boneToModify {boneName=dyn_hair_2_r}, constraintSetup [{targetBone {boneName=dyn_hair_2_c}}], constraintWeights [1]}
   n56 = "Constraint - dyn_hair_3_r (P)"!(ComponentPose=@n55, ConstraintWeights_0=1) -> (Pose) {alpha=0.3, boneToModify {boneName=dyn_hair_3_r}, constraintSetup [{targetBone {boneName=dyn_hair_3_c}}], constraintWeights [1]}
   n57 = "Constraint - dyn_hair_4_r (P)"!(ComponentPose=@n56, ConstraintWeights_0=1) -> (Pose) {alpha=0.3, boneToModify {boneName=dyn_hair_4_r}, constraintSetup [{targetBone {boneName=dyn_hair_4_c}}], constraintWeights [1]}
   n58 = "Constraint - dyn_hair_5_r (P)"!(ComponentPose=@n57, ConstraintWeights_0=1) -> (Pose) {alpha=0.3, boneToModify {boneName=dyn_hair_5_r}, constraintSetup [{targetBone {boneName=dyn_hair_5_c}}], constraintWeights [1]}
   n59 = "Constraint - dyn_hair_6_r (P)"!(ComponentPose=@n58, ConstraintWeights_0=1) -> (Pose) {boneToModify {boneName=dyn_hair_6_r}, constraintSetup [{targetBone {boneName=dyn_hair_6_c}}], constraintWeights [1]}
   n60 = "Constraint - dyn_hair_7_r (P)"!(ComponentPose=@n59, ConstraintWeights_0=1) -> (Pose) {boneToModify {boneName=dyn_hair_7_r}, constraintSetup [{targetBone {boneName=dyn_hair_7_c}}], constraintWeights [1]}
   Save cached pose 'CRCache_PlayingMeele'(Pose=Component To Local(ComponentPose=@n60)) {cacheName=CRCache_PlayingMeele}
   Save cached pose 'postSimFixes'(Pose=Use cached pose 'dynamicsCP'()) {cacheName=postSimFixes}
  ]}
 EventGraph {nodes [
   Event BlueprintUpdateAnimation()
   > n1 = Sequence!(execute) -> (then_0, then_1, then_2:Exec, then_3:Exec, then_4:Exec, then_5:Exec)
   > n2 = "Gravity Override State Function (Nativized)"!(
    execute
    CharacterPartAnimInstance:Custom Character Part Anim Instance Object Reference=Self-Reference()
    GravityOverrideParameterSet:Fort Gravity Override Parameter Set Structure=Get Gravity Override Params() -> (Gravity Override Params:Fort Gravity Override Parameter Set Structure)
    RigidBodyAnimNodeParameterSet:Fort Rigid Body Anim Node Parameter Set Structure
    DebugDraw:bool=false
    GameTimeInSeconds:float=GetGameTimeInSeconds()
    DebugJointName:Name=root
    StateLogicLayerName:Name=none
   ) -> (
    then
    OutGravityOverride:Vector
    OutCurrentState:Name=None
    OutGravityOverrideParameters:Fort Gravity Override Parameters Structure
    OutRigidBodyAnimNodeParameters:Fort Rigid Body Anim Node Parameters Structure
   )
   > Set GravityOverride(GravityOverride=@n2)
   n3 = Cast To FortPlayerPawn(execute=@n1.then_4, Object=GetOwningActor())
   > Is Valid(InputObject=Get LatteStirBackblingSkelMesh() -> (LatteStirBackblingSkelMesh:Skeletal Mesh Object Reference))
   > n4 = Set currentBackpack(currentBackpack:Athena Backpack Item Definition Object Reference=GetItemDefinitionTypeFromLoadoutWithActor!(
     Actor:Actor Object Reference=@n3
     LoadoutSlot:Gameplay Tag Structure=(TagName="LoadoutSlot.Backpack")
     CosmeticType:Athena Cosmetic Item Definition Class Reference=/Script/FortniteGame.AthenaBackpackItemDefinition
    ) -> (OutCosmeticItemDef:Athena Backpack Item Definition Object Reference)) -> (Output_Get:Athena Backpack Item Definition Object Reference)
   > Set isLatteStirBackbling(isLatteStirBackbling:bool="Equal (Object)"(A=@n4, B=Get backblingItemDef() -> (backblingItemDef:Athena Backpack Item Definition Object Reference)) {is=title K2Node_CallFunction}) -> (Output_Get:bool)
   n5 = Branch(execute=@n1.then_2, Condition=Get bIsPlayingMeleeAnim())
   > Set alpha hairRigidBody(alpha hairRigidBody:double=0.6) -> (Output_Get:double)
   n6 = Branch(execute=@n5.else, Condition=Get isTacticalJumpingFalling() -> (isTacticalJumpingFalling:bool))
   > Set alpha hairRigidBody(alpha hairRigidBody=0.8)
   n7 = Branch(execute=@n6.else, Condition=Get bIsDBNOCarrying() -> (bIsDBNOCarrying:bool))
   > Set alpha hairRigidBody(alpha hairRigidBody=0.4)
   Set alpha hairRigidBody(execute=@n7.else, alpha hairRigidBody=1)
   n8 = float * float!(
    A="Clamp (Float)"(Value=Get PawnVelocityZ(), Min=-1000, Max=0)
    B:bool=Get bRigidBodyEnabled()
    C:bool=NOT Boolean(A=Get isWearingBlackbling() -> (isWearingBlackbling:bool))
    D:bool=NOT Boolean(A=Get bIsSkydiveDiveMode())
    E:bool=NOT Boolean(A=Get bIsDBNOCarrying())
   ) -> (ReturnValue) {is=title K2Node_PromotableOperator}
   n9 = "Atan2 (Degrees)"!(Y:double=@n8, X:double=3800) -> (ReturnValue:double)
   Set hairBaseOverlap rotator(execute=@n1.then_3, hairBaseOverlap rotator:Rotator=MakeRotator(Pitch=float * float!(A=@n9, B:double=-1) -> (ReturnValue))) -> (Output_Get:Rotator)
   > Set hairBaseOverlap vector(hairBaseOverlap vector:Vector=MakeVector(Y=0, Z=Remap Float(Input=@n8, OldRange=(X=0,Y=-1000), NewRange=(X=0,Y=2), Clamp=true) -> (Output=~))) -> (Output_Get:Vector)
   n10 = Get bIsTacticalSprinting()
   n11 = Get bIsJumping()
   n12 = Get bIsFalling()
   Set isTacticalJumpingFalling(execute=@n1.then_1, isTacticalJumpingFalling:bool=OR Boolean(A=AND Boolean(A=@n11, B=@n10), B=AND Boolean(A=@n10, B=@n12))) -> (Output_Get:bool)
   n13 = PickaxeItemDefinitionHasAnyMetaTag(Tags=Make Array("[0]"=(TagName="Cosmetics.MetaTag.LatteStir")), bRequiresEquipped=true)
   n14 = Branch(execute=@n1.then_5, Condition=AND Boolean(A=@n13, B=Get isLatteStirBackbling() -> (isLatteStirBackbling:bool)))
   > n15 = Branch(Condition=AND Boolean(A=Get isWearingBlackbling(), B=Get EquipedCustomPickaxe() -> (EquipedCustomPickaxe:bool)))
   > Set TurnOffDyn(TurnOffDyn:bool=false) -> (Output_Get:bool)
   Set TurnOffDyn(execute=@n15.else, TurnOffDyn=true)
   Set TurnOffDyn(execute=@n14.else, TurnOffDyn=Get isWearingBlackbling())
   Event OnParachuteOpened()
   > Switch on EGliderType(Selection=Get PlayerGliderType() -> (PlayerGliderType=HangGlider))
   > Set isUsingHangGlidingInParachute(isUsingHangGlidingInParachute:bool=true) -> (Output_Get:bool)
   n16 = Event OnBeginTacticalSprinting() {is=title K2Node_Event}
   Set EquipedCustomPickaxe(execute=[@n16.then, @n18.then, @n20.then, @n24.then, @n21.else, @n23.else], EquipedCustomPickaxe:bool=false) -> (Output_Get:bool)
   n17 = Event OnEndTacticalSprinting() {is=title K2Node_Event}
   Set EquipedCustomPickaxe(execute=[@n17.then, @n19.then, @n24.else, @n21.then], EquipedCustomPickaxe=true)
   n18 = Event OnBeginSwimming() {is=title K2Node_Event}
   n19 = Event OnEndSwimming() {is=title K2Node_Event}
   n20 = Event OnBeginSkydiving()
   Event OnEndSkydiving()
   > n21 = Branch(Condition=@n13)
   n22 = Event OnWeaponChanged()
   Get Dynamic Material Instance(execute=[@n25.then_1, @n22.then], SkeletalMeshComponent=GetPartSkeletalMeshComponent())
   > n23 = Branch(Condition=@n13)
   > n24 = Branch(Condition=Get bIsSkyDiving())
   Event OnCharacterCustomizationLoaded()
   > n25 = Sequence!(execute) -> (then_0, then_1)
   n26 = CheckForBackbling()
   > Set isWearingBlackbling(isWearingBlackbling:bool=OR Boolean!(A=@n26, B=@n26.bIsBackblingACape, C:bool=@n26.bIsBackblingLong) -> (ReturnValue)) -> (Output_Get:bool)
   > n27 = Branch(Condition=Get bIsInFrontEnd())
   > Set SimSpaceSettings output(SimSpaceSettings output:Anim Phys Sim Space Settings Structure=Get SimSpaceSettings frontEnd() -> (SimSpaceSettings frontEnd:Anim Phys Sim Space Settings Structure)) -> (Output_Get:Anim Phys Sim Space Settings Structure)
   Set SimSpaceSettings output(
    execute=@n27.else
    SimSpaceSettings output=Get SimSpaceSettings default() -> (SimSpaceSettings default:Anim Phys Sim Space Settings Structure)
   )
   Event OnParachuteClosed()
   > Set isUsingHangGlidingInParachute(isUsingHangGlidingInParachute=false)
   Set isAR Equipped(isAR Equipped:bool=false) -> (Output_Get:bool)
   > Set isPistol Equipped(isPistol Equipped:bool=false) -> (Output_Get:bool)
   > n28 = Switch on EFortWeaponCoreAnimation!(execute, Selection:EFortWeaponCoreAnimation Enum=Get WeaponCoreAnim!() -> (WeaponCoreAnim)) -> (
    Melee:Exec
    Pistol:Exec
    Shotgun:Exec
    PaperBlueprint:Exec
    Rifle:Exec
    MeleeOneHand:Exec
    MachinePistol:Exec
    RocketLauncher:Exec
    GrenadeLauncher:Exec
    GoingCommando:Exec
    AssaultRifle:Exec
    TacticalShotgun:Exec
    SniperRifle:Exec
    TrapPlacement:Exec
    ShoulderLauncher:Exec
    AbilityDecoTool:Exec
    Crossbow:Exec
    C4:Exec
    RemoteControl:Exec
    DualWield:Exec
    AR_BullPup:Exec
    AR_ForwardGrip:Exec
    MedPackPaddles:Exec
    SMG_P90:Exec
    AR_DrumGun:Exec
    Consumable_Small:Exec
    Consumable_Large:Exec
    Balloon:Exec
    MountedTurret:Exec
    CreativeTool:Exec
    ExplosiveBow:Exec
    AshtonIndigo:Exec
    AshtonChicago:Exec
    MeleeDualWield:Exec
    Unarmed:Exec
   )
   Set isAR Equipped(execute=@n28.AssaultRifle, isAR Equipped=true)
   Set isPistol Equipped(execute=@n28.Pistol, isPistol Equipped=true)
   Get isWearingBlackbling()
  ]}
]
variables [
 alpha hairRigidBody {type=number}
 backblingItemDef {type=AthenaBackpackItemDefinition, default=/BRCosmetics/Athena/Items/Cosmetics/Backpacks/Backpack_LatteStir}
 currentBackpack {type=AthenaBackpackItemDefinition, default=null}
 EquipedCustomPickaxe {type=boolean}
 Gravity Override Params {
  type=FortGravityOverrideParameterSet
  default=(Skydive_GoingUp=(GravityOverride=(WindFrequency=(X=1.500000,Y=7.000000,Z=8.000000),WindAmplitude=(X=200.000000,Y=-500.000000,Z=400.000000),GravityOverride=(X=0.000000,Y=-300.000000,Z=1500.000000),JointName="root")),Skydive_FloatingDown=(GravityOverride=(WindFrequency=(X=1.500000,Y=7.000000,Z=8.000000),WindAmplitude=(X=200.000000,Y=-500.000000,Z=400.000000),WindOffset=(X=0.000000,Y=-600.000000,Z=400.000000),GravityOverride=(X=0.000000,Y=-700.000000,Z=600.000000),JointName="root")),Skydive_DivingDown=(GravityOverride=(WindFrequency=(X=1.500000,Y=12.000000,Z=10.000000),WindAmplitude=(X=50.000000,Y=300.000000,Z=100.000000),GravityOverride=(X=-1500.000000,Y=0.000000,Z=0.000000),JointName="pelvis")),Skydive_Parachuting=(GravityOverride=(WindFrequency=(X=1.500000,Y=7.000000,Z=8.000000),WindAmplitude=(X=200.000000,Y=10.000000,Z=50.000000),GravityOverride=(X=0.000000,Y=0.000000,Z=-1200.000000),JointName="root")),OnGround_Standing=(GravityOverride=(WindFrequency=(X=5.000000,Y=4.000000,Z=5.000000),WindAmplitude=(X=200.000000,Y=200.000000,Z=500.000000),GravityOverride=(X=-100.000000,Y=0.000000,Z=-300.000000),JointName="root")),OnGround_Moving=(GravityOverride=(GravityOverride=(X=0.000000,Y=-400.000000,Z=-400.000000),JointName="root")),OnGround_Moving_HighRotationRate=(GravityOverride=(GravityOverride=(X=0.000000,Y=0.000000,Z=-980.000000),JointName="root")),OnGround_Crouch_Moving=(GravityOverride=(GravityOverride=(X=0.000000,Y=0.000000,Z=-980.000000),JointName="root")),Emote_or_Melee=(GravityOverride=(GravityOverride=(X=0.000000,Y=0.000000,Z=-980.000000),JointName="root")),Swimming=(GravityOverride=(GravityOverride=(X=-800.000000,Y=0.000000,Z=0.000000),JointName="pelvis")),Front_End=(GravityOverride=(WindFrequency=(X=5.000000,Y=4.000000,Z=5.000000),WindAmplitude=(X=200.000000,Y=100.000000,Z=500.000000),GravityOverride=(X=-100.000000,Y=0.000000,Z=-300.000000),JointName="root")),DBNO=(GravityOverride=(GravityOverride=(X=-980.000000,Y=400.000000,Z=0.000000),JointName="pelvis")),DBNO_Carried=(GravityOverride=(GravityOverride=(X=-980.000000,Y=-500.000000,Z=-500.000000),JointName="pelvis")),OnGround_Falling=(GravityOverride=(GravityOverride=(X=0.000000,Y=0.000000,Z=-980.000000),JointName="root")),Ziplining=(GravityOverride=(GravityOverride=(X=0.000000,Y=0.000000,Z=-980.000000),JointName="root")),Skydive_Surfing=(GravityOverride=(WindFrequency=(X=1.500000,Y=7.000000,Z=8.000000),WindAmplitude=(X=200.000000,Y=-500.000000,Z=400.000000),GravityOverride=(X=500.000000,Y=-300.000000,Z=-980.000000),JointName="root")),Riding_Standing=(GravityOverride=(GravityOverride=(X=0.000000,Y=-350.000000,Z=-980.000000),JointName="root")),Riding_Moving=(GravityOverride=(GravityOverride=(X=0.000000,Y=-450.000000,Z=-450.000000),JointName="root")),Riding_Moving_Targeting=(GravityOverride=(GravityOverride=(X=0.000000,Y=-200.000000,Z=-980.000000),JointName="root")),Tactical_Sprint=(GravityOverride=(GravityOverride=(X=0.000000,Y=-450.000000,Z=-500.000000),JointName="root")),Rail_Grinding=(GravityOverride=(GravityOverride=(X=0.000000,Y=0.000000,Z=-980.000000),JointName="root")),)
 }
 GravityOverride {type=Vector}
 hairBaseOverlap rotator {type=Rotator}
 hairBaseOverlap vector {type=Vector}
 isAR Equipped {type=boolean}
 isLatteStirBackbling {type=boolean}
 isPistol Equipped {type=boolean}
 isTacticalJumpingFalling {type=boolean}
 isUsingHangGlidingInParachute {type=boolean}
 isWearingBlackbling {type=boolean}
 LatteStirBackblingSkelMesh {type=SkeletalMesh, default=/BRCosmetics/Accessories/FORT_Backpacks/Backpack_M_MED_LatteStir/Meshes/M_MED_LatteStir_Pack}
 Play MeeleFloat {type=number}
 SimSpaceSettings default {
  type=AnimPhysSimSpaceSettings
  default {maxAngularAcceleration=10000, maxAngularVelocity=10000, simSpaceAngularAlpha=0, externalAngularVelocity=<0 0 0>}
 }
 SimSpaceSettings frontEnd {
  type=AnimPhysSimSpaceSettings
  default {maxAngularAcceleration=10000, maxAngularVelocity=10000, simSpaceAngularAlpha=0, externalAngularVelocity=<0 0 0>}
 }
 SimSpaceSettings output {
  type=AnimPhysSimSpaceSettings
  default {maxAngularAcceleration=10000, maxAngularVelocity=10000, simSpaceAngularAlpha=0, externalAngularVelocity=<0 0 0>}
 }
 TurnOffDyn {type=boolean}
]
